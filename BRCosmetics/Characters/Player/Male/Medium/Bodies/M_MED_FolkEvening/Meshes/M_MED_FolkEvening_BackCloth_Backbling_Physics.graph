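# Physics asset for the back cloth backbling of the M_MED_FolkEvening male character, pairing a dangling cloth chain with a capsule body proxy.
# Capsules cover pelvis, spine, arms and legs, joined by freely moving constraints.
# The dyn_backCloth chain hangs from clavicle_l through a free master joint, then five segments with tight angular limits, loosening twist and swing toward the tip.
# The foot_r capsule duplicates clavicle_r's shape rather than mirroring foot_l.
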PhysicsAsset /BRCosmetics/Characters/Player/Male/Medium/Bodies/M_MED_FolkEvening/Meshes/M_MED_FolkEvening_BackCloth_Backbling_Physics
bodies {
 calf_l [{shapeType=Capsule, length=30.5811, radius=10, shapeName=calf_l_capsule, center=<-18.0275 0 0>, rotation=<rot 90 0 0>}]
 calf_r [{shapeType=Capsule, length=30.5811, radius=10, shapeName=calf_l_capsule, center=<18.0275 0 0>, rotation=<rot 90 0 180>}]
 clavicle_l [{shapeType=Capsule, length=14.9526, radius=11.9478, shapeName=clavicle_l_capsule, center=<11.791 -1.2106 3.2046>, rotation=<rot 90 0 0>}]
 clavicle_r [{shapeType=Capsule, length=14.9526, radius=11.9478, shapeName=clavicle_l_capsule, center=<-11.791 1.2106 -3.2046>, rotation=<rot 90 0 180>}]
 dyn_backCloth_01 [{shapeType=Capsule, length=6.2949, radius=6.1632, shapeName=dyn_backCloth_01_capsule, center=<4.6 -5 -1.8>}]
 dyn_backCloth_02 [{shapeType=Capsule, length=5.5542, radius=5.9582, shapeName=dyn_backCloth_02_capsule, center=<6.473 -5 -2>}]
 dyn_backCloth_03 [{shapeType=Capsule, length=5.0499, radius=7.0347, shapeName=dyn_backCloth_03_capsule, center=<4.6869 -4.9329 0>}]
 dyn_backCloth_04 [{shapeType=Capsule, length=1.1716, radius=12.3924, shapeName=dyn_backCloth_04_capsule, center=<6.1072 -11 0>}]
 dyn_backCloth_05 [{shapeType=Capsule, length=0.5706, radius=12.3924, shapeName=dyn_backCloth_05_capsule, center=<3.9711 -11 0>}]
 dyn_backCloth_MSTR [{shapeType=Capsule, length=0.5, radius=0.5, shapeName=dyn_backCloth_MSTR_capsule}]
 foot_l [{shapeType=Capsule, length=17.5254, radius=7.256, shapeName=foot_l_capsule, center=<-3.4185 6.0044 0>, rotation=<rot 90 180 -90>}]
 foot_r [{shapeType=Capsule, length=14.9526, radius=11.9478, shapeName=clavicle_l_capsule, center=<-11.791 1.2106 -3.2046>, rotation=<rot 90 0 180>}]
 lowerarm_l [{shapeType=Capsule, length=20.2107, radius=7.2736, shapeName=lowerarm_l_capsule, center=<14.5472 0 0>, rotation=<rot 90 0 0>}]
 lowerarm_r [{shapeType=Capsule, length=20.2107, radius=7.2736, shapeName=lowerarm_l_capsule, center=<-14.5472 0 0>, rotation=<rot 90 0 180>}]
 pelvis [{shapeType=Capsule, length=7.9318, radius=15.6144, shapeName=pelvis_capsule}]
 spine_01 [{shapeType=Capsule, length=9.4851, radius=14.0611, shapeName=spine_01_capsule}]
 spine_02 [{shapeType=Capsule, length=9.8095, radius=13.7367, shapeName=spine_02_capsule}]
 spine_03 [{shapeType=Capsule, length=10.1674, radius=13.3788, shapeName=spine_03_capsule}]
 spine_04 [{shapeType=Capsule, length=10.7838, radius=13.5406, shapeName=spine_04_capsule}]
 spine_05 [{shapeType=Capsule, length=13.173, radius=14.3225, shapeName=spine_05_capsule}]
 thigh_l [{shapeType=Capsule, length=21.3439, radius=13.8798, shapeName=thigh_l_capsule, center=<-16.551 0 0>, rotation=<rot 87.3877 -5.9067 0.0189>}]
 thigh_r [{shapeType=Capsule, length=22.5004, radius=12.7234, shapeName=thigh_l_capsule, center=<16.551 0 0>, rotation=<rot 87.3877 -5.9067 -179.9811>}]
 upperarm_l [{shapeType=Capsule, length=20.2107, radius=7.2736, shapeName=upperarm_l_capsule, center=<14.5472 0 0>, rotation=<rot 90 0 0>}]
 upperarm_r [{shapeType=Capsule, length=20.2107, radius=7.2736, shapeName=upperarm_l_capsule, center=<-14.5472 0 0>, rotation=<rot 90 0 180>}]
}
constraints [
 {
  bone1Name=spine_01
  bone2Name=pelvis
  swing1LimitDegrees=45
  swing1Motion=Free
  swing2LimitDegrees=45
  swing2Motion=Free
  twistLimitDegrees=45
  twistMotion=Free
 }
 {
  bone1Name=spine_02
  bone2Name=spine_01
  swing1LimitDegrees=45
  swing1Motion=Free
  swing2LimitDegrees=45
  swing2Motion=Free
  twistLimitDegrees=45
  twistMotion=Free
 }
 {
  bone1Name=spine_03
  bone2Name=spine_02
  swing1LimitDegrees=45
  swing1Motion=Free
  swing2LimitDegrees=45
  swing2Motion=Free
  twistLimitDegrees=45
  twistMotion=Free
 }
 {
  bone1Name=spine_04
  bone2Name=spine_03
  swing1LimitDegrees=45
  swing1Motion=Free
  swing2LimitDegrees=45
  swing2Motion=Free
  twistLimitDegrees=45
  twistMotion=Free
 }
 {
  bone1Name=spine_05
  bone2Name=spine_04
  swing1LimitDegrees=45
  swing1Motion=Free
  swing2LimitDegrees=45
  swing2Motion=Free
  twistLimitDegrees=45
  twistMotion=Free
 }
 {
  bone1Name=thigh_l
  bone2Name=pelvis
  swing1LimitDegrees=45
  swing1Motion=Free
  swing2LimitDegrees=45
  swing2Motion=Free
  twistLimitDegrees=45
  twistMotion=Free
 }
 {
  bone1Name=calf_l
  bone2Name=thigh_l
  swing1LimitDegrees=45
  swing1Motion=Free
  swing2LimitDegrees=45
  swing2Motion=Free
  twistLimitDegrees=45
  twistMotion=Free
 }
 {
  bone1Name=foot_l
  bone2Name=calf_l
  swing1LimitDegrees=45
  swing1Motion=Free
  swing2LimitDegrees=45
  swing2Motion=Free
  twistLimitDegrees=45
  twistMotion=Free
 }
 {
  bone1Name=clavicle_l
  bone2Name=spine_05
  swing1LimitDegrees=45
  swing1Motion=Free
  swing2LimitDegrees=45
  swing2Motion=Free
  twistLimitDegrees=45
  twistMotion=Free
 }
 {
  bone1Name=upperarm_l
  bone2Name=clavicle_l
  swing1LimitDegrees=45
  swing1Motion=Free
  swing2LimitDegrees=45
  swing2Motion=Free
  twistLimitDegrees=45
  twistMotion=Free
 }
 {
  bone1Name=lowerarm_l
  bone2Name=upperarm_l
  swing1LimitDegrees=45
  swing1Motion=Free
  swing2LimitDegrees=45
  swing2Motion=Free
  twistLimitDegrees=45
  twistMotion=Free
 }
 {
  bone1Name=clavicle_r
  bone2Name=spine_05
  swing1LimitDegrees=45
  swing1Motion=Free
  swing2LimitDegrees=45
  swing2Motion=Free
  twistLimitDegrees=45
  twistMotion=Free
 }
 {
  bone1Name=upperarm_r
  bone2Name=clavicle_r
  swing1LimitDegrees=45
  swing1Motion=Free
  swing2LimitDegrees=45
  swing2Motion=Free
  twistLimitDegrees=45
  twistMotion=Free
 }
 {
  bone1Name=lowerarm_r
  bone2Name=upperarm_r
  swing1LimitDegrees=45
  swing1Motion=Free
  swing2LimitDegrees=45
  swing2Motion=Free
  twistLimitDegrees=45
  twistMotion=Free
 }
 {
  bone1Name=thigh_r
  bone2Name=pelvis
  swing1LimitDegrees=45
  swing1Motion=Free
  swing2LimitDegrees=45
  swing2Motion=Free
  twistLimitDegrees=45
  twistMotion=Free
 }
 {
  bone1Name=calf_r
  bone2Name=thigh_r
  swing1LimitDegrees=45
  swing1Motion=Free
  swing2LimitDegrees=45
  swing2Motion=Free
  twistLimitDegrees=45
  twistMotion=Free
 }
 {
  bone1Name=foot_r
  bone2Name=calf_r
  swing1LimitDegrees=45
  swing1Motion=Free
  swing2LimitDegrees=45
  swing2Motion=Free
  twistLimitDegrees=45
  twistMotion=Free
 }
 {
  bone1Name=dyn_backCloth_MSTR
  bone2Name=clavicle_l
  swing1LimitDegrees=45
  swing1Motion=Free
  swing2LimitDegrees=45
  swing2Motion=Free
  twistLimitDegrees=45
  twistMotion=Free
 }
 {bone1Name=dyn_backCloth_01, bone2Name=dyn_backCloth_MSTR, swing1LimitDegrees=5, swing2LimitDegrees=20, twistLimitDegrees=0}
 {bone1Name=dyn_backCloth_02, bone2Name=dyn_backCloth_01, swing1LimitDegrees=5, swing2LimitDegrees=45, twistLimitDegrees=0}
 {bone1Name=dyn_backCloth_03, bone2Name=dyn_backCloth_02, swing1LimitDegrees=0, swing2LimitDegrees=10, twistLimitDegrees=0}
 {bone1Name=dyn_backCloth_04, bone2Name=dyn_backCloth_03, swing1LimitDegrees=30, swing2LimitDegrees=45, twistLimitDegrees=5}
 {bone1Name=dyn_backCloth_05, bone2Name=dyn_backCloth_04, swing1LimitDegrees=30, swing2LimitDegrees=45, twistLimitDegrees=5}
]
