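# Physics asset for the back cloth of the male medium FolkEvening player body cosmetic.
# A four-link capsule chain hangs from a master bone constrained to the left clavicle, with swing and twist limits that generally loosen toward the tip.
# Capsules on the pelvis, spine, left thigh and left arm model the surrounding body, joined by free-motion constraints.
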PhysicsAsset /BRCosmetics/Characters/Player/Male/Medium/Bodies/M_MED_FolkEvening/Meshes/M_MED_FolkEvening_BackCloth_Physics
bodies {
 clavicle_l [{shapeType=Capsule, length=14.9526, radius=11.9478, shapeName=clavicle_l_capsule, center=<11.791 -1.2106 3.2046>, rotation=<rot 90 0 0>}]
 dyn_backCloth_01 [{shapeType=Capsule, length=6.2949, radius=6.1632, shapeName=dyn_backCloth_01_capsule, center=<4.6 -5 -1.8>}]
 dyn_backCloth_02 [{shapeType=Capsule, length=6.7811, radius=6.8301, shapeName=dyn_backCloth_02_capsule, center=<6.473 -5 -2>}]
 dyn_backCloth_03 [{shapeType=Capsule, length=0.7912, radius=13.4704, shapeName=dyn_backCloth_03_capsule, center=<4.6869 -11 0>}]
 dyn_backCloth_04 [{shapeType=Capsule, length=15.4545, radius=11.1819, shapeName=dyn_backCloth_04_capsule, center=<8.9357 -8 0>, rotation=<rot 90 0 0>}]
 dyn_backCloth_MSTR [{shapeType=Capsule, length=0.5, radius=0.5, shapeName=dyn_backCloth_MSTR_capsule}]
 lowerarm_l [{shapeType=Capsule, length=20.2107, radius=7.2736, shapeName=lowerarm_l_capsule, center=<14.5472 0 0>, rotation=<rot 90 0 0>}]
 pelvis [
  {shapeType=Capsule, length=10.5357, radius=16.9532, shapeName=pelvis_capsule}
  {shapeType=Capsule, length=11.896, radius=18.104, shapeName=pelvis_capsule, center=<-11.076 1.892 0>}
 ]
 spine_01 [{shapeType=Capsule, length=11.8002, radius=15.6887, shapeName=spine_01_capsule}]
 spine_02 [{shapeType=Capsule, length=9.0291, radius=14.5171, shapeName=spine_02_capsule}]
 spine_03 [{shapeType=Capsule, length=9.7456, radius=13.8006, shapeName=spine_03_capsule}]
 spine_04 [{shapeType=Capsule, length=10.7838, radius=13.5406, shapeName=spine_04_capsule}]
 spine_05 [{shapeType=Capsule, length=13.173, radius=14.3225, shapeName=spine_05_capsule}]
 thigh_l [{shapeType=Capsule, length=34.1527, radius=14.091, shapeName=thigh_l_capsule, center=<-16.551 0 0>, rotation=<rot 87.3877 -5.9067 0.0189>}]
 upperarm_l [{shapeType=Capsule, length=20.2107, radius=7.2736, shapeName=upperarm_l_capsule, center=<14.5472 0 0>, rotation=<rot 90 0 0>}]
}
constraints [
 {
  bone1Name=spine_01
  bone2Name=pelvis
  swing1LimitDegrees=45
  swing1Motion=Free
  swing2LimitDegrees=45
  swing2Motion=Free
  twistLimitDegrees=45
  twistMotion=Free
 }
 {
  bone1Name=spine_02
  bone2Name=spine_01
  swing1LimitDegrees=45
  swing1Motion=Free
  swing2LimitDegrees=45
  swing2Motion=Free
  twistLimitDegrees=45
  twistMotion=Free
 }
 {
  bone1Name=spine_03
  bone2Name=spine_02
  swing1LimitDegrees=45
  swing1Motion=Free
  swing2LimitDegrees=45
  swing2Motion=Free
  twistLimitDegrees=45
  twistMotion=Free
 }
 {
  bone1Name=spine_04
  bone2Name=spine_03
  swing1LimitDegrees=45
  swing1Motion=Free
  swing2LimitDegrees=45
  swing2Motion=Free
  twistLimitDegrees=45
  twistMotion=Free
 }
 {
  bone1Name=spine_05
  bone2Name=spine_04
  swing1LimitDegrees=45
  swing1Motion=Free
  swing2LimitDegrees=45
  swing2Motion=Free
  twistLimitDegrees=45
  twistMotion=Free
 }
 {
  bone1Name=thigh_l
  bone2Name=pelvis
  swing1LimitDegrees=45
  swing1Motion=Free
  swing2LimitDegrees=45
  swing2Motion=Free
  twistLimitDegrees=45
  twistMotion=Free
 }
 {
  bone1Name=clavicle_l
  bone2Name=spine_05
  swing1LimitDegrees=45
  swing1Motion=Free
  swing2LimitDegrees=45
  swing2Motion=Free
  twistLimitDegrees=45
  twistMotion=Free
 }
 {
  bone1Name=upperarm_l
  bone2Name=clavicle_l
  swing1LimitDegrees=45
  swing1Motion=Free
  swing2LimitDegrees=45
  swing2Motion=Free
  twistLimitDegrees=45
  twistMotion=Free
 }
 {
  bone1Name=lowerarm_l
  bone2Name=upperarm_l
  swing1LimitDegrees=45
  swing1Motion=Free
  swing2LimitDegrees=45
  swing2Motion=Free
  twistLimitDegrees=45
  twistMotion=Free
 }
 {
  bone1Name=dyn_backCloth_MSTR
  bone2Name=clavicle_l
  swing1LimitDegrees=45
  swing1Motion=Free
  swing2LimitDegrees=45
  swing2Motion=Free
  twistLimitDegrees=45
  twistMotion=Free
 }
 {bone1Name=dyn_backCloth_01, bone2Name=dyn_backCloth_MSTR, swing1LimitDegrees=5, swing2LimitDegrees=10, twistLimitDegrees=0}
 {bone1Name=dyn_backCloth_02, bone2Name=dyn_backCloth_01, swing1LimitDegrees=15, swing2LimitDegrees=21, twistLimitDegrees=0}
 {bone1Name=dyn_backCloth_03, bone2Name=dyn_backCloth_02, swing1LimitDegrees=25, swing2LimitDegrees=20, twistLimitDegrees=10}
 {bone1Name=dyn_backCloth_04, bone2Name=dyn_backCloth_03, swing1LimitDegrees=25, swing2LimitDegrees=25, twistLimitDegrees=20}
]
