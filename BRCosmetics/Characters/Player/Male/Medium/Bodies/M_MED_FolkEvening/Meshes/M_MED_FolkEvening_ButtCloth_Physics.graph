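# Physics asset for the male medium FolkEvening cosmetic body, setting up bone-chain physics for its front and back loincloths.
# Capsules on torso, arms and legs form the body collision, with a huge offset root capsule whose top sits near the origin like a floor.
# Two five-bone cloth chains hang from pelvis-linked master bones with limited swing and twist, generally loosening toward the tips.
# Body and master-bone constraints leave swing and twist free.
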PhysicsAsset /BRCosmetics/Characters/Player/Male/Medium/Bodies/M_MED_FolkEvening/Meshes/M_MED_FolkEvening_ButtCloth_Physics
bodies {
 calf_l [{shapeType=Capsule, length=39.9473, radius=10, shapeName=calf_l_capsule, center=<-18.0275 0 0>, rotation=<rot 90 0 0>}]
 calf_r [{shapeType=Capsule, length=40.1495, radius=10, shapeName=calf_l_capsule, center=<18.0275 0 0>, rotation=<rot 90 0 180>}]
 clavicle_l [{shapeType=Capsule, length=14.9526, radius=11.9478, shapeName=clavicle_l_capsule, center=<11.791 -1.2106 3.2046>, rotation=<rot 90 0 0>}]
 clavicle_r [{shapeType=Capsule, length=14.9526, radius=11.9478, shapeName=clavicle_l_capsule, center=<-11.791 1.2106 -3.2046>, rotation=<rot 90 0 180>}]
 dyn_backCloth_MSTR [{shapeType=Capsule, length=0.5, radius=0.5, shapeName=dyn_backCloth_MSTR_capsule}]
 dyn_backLoinCloth_01 [{shapeType=Capsule, length=15.8302, radius=6.9544, shapeName=dyn_backLoinCloth_01_capsule, center=<8.8338 -6 0>}]
 dyn_backLoinCloth_02 [{shapeType=Capsule, length=8.9448, radius=8.4193, shapeName=dyn_backLoinCloth_02_capsule, center=<8.8338 -6 0>}]
 dyn_backLoinCloth_03 [{shapeType=Capsule, length=6.5719, radius=9.159, shapeName=dyn_backLoinCloth_03_capsule, center=<8.8338 -6 0>}]
 dyn_backLoinCloth_04 [{shapeType=Capsule, length=6.231, radius=8.3599, shapeName=dyn_backLoinCloth_04_capsule, center=<8.8338 -6 0>}]
 dyn_backLoinCloth_05 [{shapeType=Capsule, length=10.3815, radius=11.2961, shapeName=dyn_backLoinCloth_05_capsule, center=<9.4277 -8 0>, rotation=<rot 90 0 0>}]
 dyn_backLoinCloth_MSTR [{shapeType=Capsule, length=0.5, radius=0.5, shapeName=dyn_backLoinCloth_MSTR_capsule}]
 dyn_loinCloth_01 [{shapeType=Capsule, length=10.854, radius=5.1294, shapeName=dyn_backLoinCloth_01_capsule, center=<7.1613 5 0>}]
 dyn_loinCloth_02 [{shapeType=Capsule, length=10.7712, radius=5.482, shapeName=dyn_backLoinCloth_02_capsule, center=<5.8185 5 0>}]
 dyn_loinCloth_03 [{shapeType=Capsule, length=8.6852, radius=5.9504, shapeName=dyn_backLoinCloth_03_capsule, center=<5.8185 5 0>}]
 dyn_loinCloth_04 [{shapeType=Capsule, length=4.0358, radius=7.0589, shapeName=dyn_backLoinCloth_04_capsule, center=<5.8185 6 0>}]
 dyn_loinCloth_05 [{shapeType=Capsule, length=2.764, radius=7.611, shapeName=dyn_backLoinCloth_05_capsule, center=<2.6393 6 0>, rotation=<rot 90 0 0>}]
 dyn_loinCloth_MSTR [{shapeType=Capsule, length=0.5, radius=0.5, shapeName=dyn_backLoinCloth_MSTR_capsule}]
 foot_l [{shapeType=Capsule, length=17.5254, radius=7.256, shapeName=foot_l_capsule, center=<-3.4185 6.0044 0>, rotation=<rot 90 180 -90>}]
 foot_r [{shapeType=Capsule, length=17.5254, radius=7.256, shapeName=foot_l_capsule, center=<3.4185 -6.0044 0>, rotation=<rot 90 153.4349 63.4349>}]
 lowerarm_l [{shapeType=Capsule, length=20.2107, radius=7.2736, shapeName=lowerarm_l_capsule, center=<14.5472 0 0>, rotation=<rot 90 0 0>}]
 lowerarm_r [{shapeType=Capsule, length=20.2107, radius=7.2736, shapeName=lowerarm_l_capsule, center=<-14.5472 0 0>, rotation=<rot 90 0 180>}]
 pelvis [{shapeType=Capsule, length=10.3702, radius=13.176, shapeName=pelvis_capsule}]
 root [{shapeType=Capsule, length=0, radius=300, shapeName=root_capsule, center=<0 0 -295.6927>, rotation=<rot 0 -90 0>}]
 spine_01 [{shapeType=Capsule, length=9.4851, radius=14.0611, shapeName=spine_01_capsule}]
 spine_02 [{shapeType=Capsule, length=9.8095, radius=13.7367, shapeName=spine_02_capsule}]
 spine_03 [{shapeType=Capsule, length=10.1674, radius=13.3788, shapeName=spine_03_capsule}]
 spine_04 [{shapeType=Capsule, length=10.7838, radius=13.5406, shapeName=spine_04_capsule}]
 spine_05 [{shapeType=Capsule, length=13.173, radius=14.3225, shapeName=spine_05_capsule}]
 thigh_l [{shapeType=Capsule, length=24.0425, radius=11.1813, shapeName=thigh_l_capsule, center=<-16.551 0 0>, rotation=<rot 87.3877 -5.9067 0.0189>}]
 thigh_r [{shapeType=Capsule, length=24.0425, radius=11.1813, shapeName=thigh_l_capsule, center=<16.551 0 0>, rotation=<rot 87.3877 -5.9067 -179.9811>}]
 upperarm_l [{shapeType=Capsule, length=20.2107, radius=7.2736, shapeName=upperarm_l_capsule, center=<14.5472 0 0>, rotation=<rot 90 0 0>}]
 upperarm_r [{shapeType=Capsule, length=20.2107, radius=7.2736, shapeName=upperarm_l_capsule, center=<-14.5472 0 0>, rotation=<rot 90 0 180>}]
}
constraints [
 {
  bone1Name=spine_01
  bone2Name=pelvis
  swing1LimitDegrees=45
  swing1Motion=Free
  swing2LimitDegrees=45
  swing2Motion=Free
  twistLimitDegrees=45
  twistMotion=Free
 }
 {
  bone1Name=spine_02
  bone2Name=spine_01
  swing1LimitDegrees=45
  swing1Motion=Free
  swing2LimitDegrees=45
  swing2Motion=Free
  twistLimitDegrees=45
  twistMotion=Free
 }
 {
  bone1Name=spine_03
  bone2Name=spine_02
  swing1LimitDegrees=45
  swing1Motion=Free
  swing2LimitDegrees=45
  swing2Motion=Free
  twistLimitDegrees=45
  twistMotion=Free
 }
 {
  bone1Name=spine_04
  bone2Name=spine_03
  swing1LimitDegrees=45
  swing1Motion=Free
  swing2LimitDegrees=45
  swing2Motion=Free
  twistLimitDegrees=45
  twistMotion=Free
 }
 {
  bone1Name=spine_05
  bone2Name=spine_04
  swing1LimitDegrees=45
  swing1Motion=Free
  swing2LimitDegrees=45
  swing2Motion=Free
  twistLimitDegrees=45
  twistMotion=Free
 }
 {
  bone1Name=thigh_l
  bone2Name=pelvis
  swing1LimitDegrees=45
  swing1Motion=Free
  swing2LimitDegrees=45
  swing2Motion=Free
  twistLimitDegrees=45
  twistMotion=Free
 }
 {
  bone1Name=calf_l
  bone2Name=thigh_l
  swing1LimitDegrees=45
  swing1Motion=Free
  swing2LimitDegrees=45
  swing2Motion=Free
  twistLimitDegrees=45
  twistMotion=Free
 }
 {
  bone1Name=foot_l
  bone2Name=calf_l
  swing1LimitDegrees=45
  swing1Motion=Free
  swing2LimitDegrees=45
  swing2Motion=Free
  twistLimitDegrees=45
  twistMotion=Free
 }
 {
  bone1Name=clavicle_l
  bone2Name=spine_05
  swing1LimitDegrees=45
  swing1Motion=Free
  swing2LimitDegrees=45
  swing2Motion=Free
  twistLimitDegrees=45
  twistMotion=Free
 }
 {
  bone1Name=upperarm_l
  bone2Name=clavicle_l
  swing1LimitDegrees=45
  swing1Motion=Free
  swing2LimitDegrees=45
  swing2Motion=Free
  twistLimitDegrees=45
  twistMotion=Free
 }
 {
  bone1Name=lowerarm_l
  bone2Name=upperarm_l
  swing1LimitDegrees=45
  swing1Motion=Free
  swing2LimitDegrees=45
  swing2Motion=Free
  twistLimitDegrees=45
  twistMotion=Free
 }
 {
  bone1Name=clavicle_r
  bone2Name=spine_05
  swing1LimitDegrees=45
  swing1Motion=Free
  swing2LimitDegrees=45
  swing2Motion=Free
  twistLimitDegrees=45
  twistMotion=Free
 }
 {
  bone1Name=upperarm_r
  bone2Name=clavicle_r
  swing1LimitDegrees=45
  swing1Motion=Free
  swing2LimitDegrees=45
  swing2Motion=Free
  twistLimitDegrees=45
  twistMotion=Free
 }
 {
  bone1Name=lowerarm_r
  bone2Name=upperarm_r
  swing1LimitDegrees=45
  swing1Motion=Free
  swing2LimitDegrees=45
  swing2Motion=Free
  twistLimitDegrees=45
  twistMotion=Free
 }
 {
  bone1Name=thigh_r
  bone2Name=pelvis
  swing1LimitDegrees=45
  swing1Motion=Free
  swing2LimitDegrees=45
  swing2Motion=Free
  twistLimitDegrees=45
  twistMotion=Free
 }
 {
  bone1Name=calf_r
  bone2Name=thigh_r
  swing1LimitDegrees=45
  swing1Motion=Free
  swing2LimitDegrees=45
  swing2Motion=Free
  twistLimitDegrees=45
  twistMotion=Free
 }
 {
  bone1Name=dyn_backCloth_MSTR
  bone2Name=clavicle_l
  swing1LimitDegrees=45
  swing1Motion=Free
  swing2LimitDegrees=45
  swing2Motion=Free
  twistLimitDegrees=45
  twistMotion=Free
 }
 {
  bone1Name=dyn_backLoinCloth_MSTR
  bone2Name=pelvis
  swing1LimitDegrees=45
  swing1Motion=Free
  swing2LimitDegrees=45
  swing2Motion=Free
  twistLimitDegrees=45
  twistMotion=Free
 }
 {bone1Name=dyn_backLoinCloth_01, bone2Name=dyn_backLoinCloth_MSTR, swing1LimitDegrees=20, swing2LimitDegrees=8, twistLimitDegrees=10}
 {bone1Name=dyn_backLoinCloth_02, bone2Name=dyn_backLoinCloth_01, swing1LimitDegrees=20, swing2LimitDegrees=10, twistLimitDegrees=10}
 {bone1Name=dyn_backLoinCloth_03, bone2Name=dyn_backLoinCloth_02, swing1LimitDegrees=30, swing2LimitDegrees=15, twistLimitDegrees=10}
 {bone1Name=dyn_backLoinCloth_04, bone2Name=dyn_backLoinCloth_03, swing1LimitDegrees=30, swing2LimitDegrees=30, twistLimitDegrees=10}
 {bone1Name=dyn_backLoinCloth_05, bone2Name=dyn_backLoinCloth_04, swing1LimitDegrees=45, swing2LimitDegrees=30, twistLimitDegrees=10}
 {
  bone1Name=foot_r
  bone2Name=calf_r
  swing1LimitDegrees=45
  swing1Motion=Free
  swing2LimitDegrees=45
  swing2Motion=Free
  twistLimitDegrees=45
  twistMotion=Free
 }
 {bone1Name=dyn_loinCloth_05, bone2Name=dyn_loinCloth_04, swing1LimitDegrees=45, swing2LimitDegrees=20, twistLimitDegrees=20}
 {
  bone1Name=dyn_loinCloth_MSTR
  bone2Name=pelvis
  swing1LimitDegrees=45
  swing1Motion=Free
  swing2LimitDegrees=45
  swing2Motion=Free
  twistLimitDegrees=45
  twistMotion=Free
 }
 {bone1Name=dyn_loinCloth_01, bone2Name=dyn_loinCloth_MSTR, swing1LimitDegrees=20, swing2LimitDegrees=20, twistLimitDegrees=20}
 {bone1Name=dyn_loinCloth_02, bone2Name=dyn_loinCloth_01, swing1LimitDegrees=20, swing2LimitDegrees=20, twistLimitDegrees=20}
 {bone1Name=dyn_loinCloth_03, bone2Name=dyn_loinCloth_02, swing1LimitDegrees=20, swing2LimitDegrees=20, twistLimitDegrees=20}
 {bone1Name=dyn_loinCloth_04, bone2Name=dyn_loinCloth_03, swing1LimitDegrees=30, swing2LimitDegrees=20, twistLimitDegrees=20}
 {
  bone1Name=pelvis
  bone2Name=root
  swing1LimitDegrees=45
  swing1Motion=Free
  swing2LimitDegrees=45
  swing2Motion=Free
  twistLimitDegrees=45
  twistMotion=Free
 }
]
