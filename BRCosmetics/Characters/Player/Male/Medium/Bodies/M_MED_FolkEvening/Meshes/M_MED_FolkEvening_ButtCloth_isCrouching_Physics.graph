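# Physics asset for the crouching variant of the male medium FolkEvening body's butt cloth, driving front and back loincloth bone chains.
# Capsules on pelvis, lower spine, legs and feet, plus a huge root capsule under the character resembling a floor, form the body shapes.
# Each five-bone cloth chain hangs from a small master bone on the pelvis with limited swing and twist, while body joints are left free.
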PhysicsAsset /BRCosmetics/Characters/Player/Male/Medium/Bodies/M_MED_FolkEvening/Meshes/M_MED_FolkEvening_ButtCloth_isCrouching_Physics
bodies {
 calf_l [{shapeType=Capsule, length=34.7659, radius=11.7463, shapeName=calf_l_capsule, center=<-18.0275 0 0>, rotation=<rot 90 0 0>}]
 calf_r [{shapeType=Capsule, length=34.7659, radius=11.7463, shapeName=calf_l_capsule, center=<18.0275 0 0>, rotation=<rot 90 0 180>}]
 dyn_backLoinCloth_01 [{shapeType=Capsule, length=15.8302, radius=6.9544, shapeName=dyn_backLoinCloth_01_capsule, center=<8.8338 -6 0>}]
 dyn_backLoinCloth_02 [{shapeType=Capsule, length=8.9448, radius=8.4193, shapeName=dyn_backLoinCloth_02_capsule, center=<8.8338 -6 0>}]
 dyn_backLoinCloth_03 [{shapeType=Capsule, length=6.5719, radius=9.159, shapeName=dyn_backLoinCloth_03_capsule, center=<8.8338 -6 0>}]
 dyn_backLoinCloth_04 [{shapeType=Capsule, length=6.231, radius=8.3599, shapeName=dyn_backLoinCloth_04_capsule, center=<8.8338 -6 0>}]
 dyn_backLoinCloth_05 [{shapeType=Capsule, length=10.3815, radius=11.2961, shapeName=dyn_backLoinCloth_05_capsule, center=<9.4277 -8 0>, rotation=<rot 90 0 0>}]
 dyn_backLoinCloth_MSTR [{shapeType=Capsule, length=0.5, radius=0.5, shapeName=dyn_backLoinCloth_MSTR_capsule}]
 dyn_loinCloth_01 [{shapeType=Capsule, length=10.854, radius=5.1294, shapeName=dyn_backLoinCloth_01_capsule, center=<7.1613 5 0>}]
 dyn_loinCloth_02 [{shapeType=Capsule, length=8.9685, radius=5.0394, shapeName=dyn_backLoinCloth_02_capsule, center=<5.8185 5 0>}]
 dyn_loinCloth_03 [{shapeType=Capsule, length=5.9353, radius=5.5917, shapeName=dyn_backLoinCloth_03_capsule, center=<5.8185 5 0>}]
 dyn_loinCloth_04 [{shapeType=Capsule, length=4.9551, radius=6.1396, shapeName=dyn_backLoinCloth_04_capsule, center=<5.8185 6 0>}]
 dyn_loinCloth_05 [{shapeType=Capsule, length=5.764, radius=6.611, shapeName=dyn_backLoinCloth_05_capsule, center=<2.6393 6 0>, rotation=<rot 90 0 0>}]
 dyn_loinCloth_MSTR [{shapeType=Capsule, length=0.5, radius=0.5, shapeName=dyn_backLoinCloth_MSTR_capsule}]
 foot_l [{shapeType=Capsule, length=17.5254, radius=7.256, shapeName=foot_l_capsule, center=<-3.4185 6.0044 0>, rotation=<rot 90 180 -90>}]
 foot_r [{shapeType=Capsule, length=17.5254, radius=7.256, shapeName=foot_l_capsule, center=<3.4185 -6.0044 0>, rotation=<rot 90 153.4349 63.4349>}]
 pelvis [{shapeType=Capsule, length=10.3702, radius=13.176, shapeName=pelvis_capsule}]
 root [{shapeType=Capsule, length=0, radius=300, shapeName=root_capsule, center=<0 0 -295.6927>, rotation=<rot 0 -90 0>}]
 spine_01 [{shapeType=Capsule, length=9.4851, radius=14.0611, shapeName=spine_01_capsule}]
 thigh_l [{
   shapeType=Capsule
   length=36.6438
   radius=11.1813
   shapeName=thigh_l_capsule
   center=<-20.7229 0.433 0.1914>
   rotation=<rot 87.3877 -5.9067 0.0189>
  }]
 thigh_r [{
   shapeType=Capsule
   length=36.6438
   radius=11.1813
   shapeName=thigh_l_capsule
   center=<20.7229 -0.433 -0.1914>
   rotation=<rot 87.3877 -5.9067 -179.9811>
  }]
}
constraints [
 {
  bone1Name=spine_01
  bone2Name=pelvis
  swing1LimitDegrees=45
  swing1Motion=Free
  swing2LimitDegrees=45
  swing2Motion=Free
  twistLimitDegrees=45
  twistMotion=Free
 }
 {
  bone1Name=thigh_l
  bone2Name=pelvis
  swing1LimitDegrees=45
  swing1Motion=Free
  swing2LimitDegrees=45
  swing2Motion=Free
  twistLimitDegrees=45
  twistMotion=Free
 }
 {
  bone1Name=calf_l
  bone2Name=thigh_l
  swing1LimitDegrees=45
  swing1Motion=Free
  swing2LimitDegrees=45
  swing2Motion=Free
  twistLimitDegrees=45
  twistMotion=Free
 }
 {
  bone1Name=foot_l
  bone2Name=calf_l
  swing1LimitDegrees=45
  swing1Motion=Free
  swing2LimitDegrees=45
  swing2Motion=Free
  twistLimitDegrees=45
  twistMotion=Free
 }
 {
  bone1Name=thigh_r
  bone2Name=pelvis
  swing1LimitDegrees=45
  swing1Motion=Free
  swing2LimitDegrees=45
  swing2Motion=Free
  twistLimitDegrees=45
  twistMotion=Free
 }
 {
  bone1Name=dyn_backLoinCloth_MSTR
  bone2Name=pelvis
  swing1LimitDegrees=45
  swing1Motion=Free
  swing2LimitDegrees=45
  swing2Motion=Free
  twistLimitDegrees=45
  twistMotion=Free
 }
 {bone1Name=dyn_backLoinCloth_01, bone2Name=dyn_backLoinCloth_MSTR, swing1LimitDegrees=31, swing2LimitDegrees=0, twistLimitDegrees=10}
 {bone1Name=dyn_backLoinCloth_02, bone2Name=dyn_backLoinCloth_01, swing1LimitDegrees=30, swing2LimitDegrees=20, twistLimitDegrees=10}
 {bone1Name=dyn_backLoinCloth_03, bone2Name=dyn_backLoinCloth_02, swing1LimitDegrees=30, swing2LimitDegrees=20, twistLimitDegrees=10}
 {bone1Name=dyn_backLoinCloth_04, bone2Name=dyn_backLoinCloth_03, swing1LimitDegrees=30, swing2LimitDegrees=30, twistLimitDegrees=10}
 {bone1Name=dyn_backLoinCloth_05, bone2Name=dyn_backLoinCloth_04, swing1LimitDegrees=45, swing2LimitDegrees=30, twistLimitDegrees=10}
 {
  bone1Name=foot_r
  bone2Name=calf_r
  swing1LimitDegrees=45
  swing1Motion=Free
  swing2LimitDegrees=45
  swing2Motion=Free
  twistLimitDegrees=45
  twistMotion=Free
 }
 {bone1Name=dyn_loinCloth_05, bone2Name=dyn_loinCloth_04, swing1LimitDegrees=30, swing2LimitDegrees=20, twistLimitDegrees=20}
 {
  bone1Name=dyn_loinCloth_MSTR
  bone2Name=pelvis
  swing1LimitDegrees=45
  swing1Motion=Free
  swing2LimitDegrees=45
  swing2Motion=Free
  twistLimitDegrees=45
  twistMotion=Free
 }
 {bone1Name=dyn_loinCloth_01, bone2Name=dyn_loinCloth_MSTR, swing1LimitDegrees=20, swing2LimitDegrees=20, twistLimitDegrees=20}
 {bone1Name=dyn_loinCloth_02, bone2Name=dyn_loinCloth_01, swing1LimitDegrees=20, swing2LimitDegrees=20, twistLimitDegrees=20}
 {bone1Name=dyn_loinCloth_03, bone2Name=dyn_loinCloth_02, swing1LimitDegrees=20, swing2LimitDegrees=20, twistLimitDegrees=20}
 {bone1Name=dyn_loinCloth_04, bone2Name=dyn_loinCloth_03, swing1LimitDegrees=30, swing2LimitDegrees=20, twistLimitDegrees=20}
 {
  bone1Name=pelvis
  bone2Name=root
  swing1LimitDegrees=45
  swing1Motion=Free
  swing2LimitDegrees=45
  swing2Motion=Free
  twistLimitDegrees=45
  twistMotion=Free
 }
 {
  bone1Name=calf_r
  bone2Name=thigh_r
  swing1LimitDegrees=45
  swing1Motion=Free
  swing2LimitDegrees=45
  swing2Motion=Free
  twistLimitDegrees=45
  twistMotion=Free
 }
]
