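# Post-process style cosmetic AnimBP for the FolkEvening male body part, adding secondary motion to its cloth and shoulder pads.
# The parent mesh pose feeds a Control Rig, backbling-switched back cloth RigidBody sims, shoulder pad bone offsets, then blended standard and crouch loincloth sims.
# The EventGraph derives movement, weapon and backbling states, varies simulation inertia scales when sprinting, and updates material WPO parameters from bone motion.
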AnimBlueprint /BRCosmetics/Characters/Player/Male/Medium/Bodies/M_MED_FolkEvening/Meshes/M_MED_FolkEvening_AnimBP
target_skeleton=&/M_MED_FolkEvening_Skeleton
graphs [
 AnimGraph {nodes [
   "Anim Dynamics - Chain: dyn_backCloth_03 -> dyn_backCloth_04"(Alpha=1, ExternalForce="0,0,0") {
    angularDampingOverride=1
    angularSpringConstant=20000
    bAngularSpring=true
    bChain=true
    bOverrideAngularDamping=true
    bUseGravityOverride=true
    bUsePlanarLimit=false
    bUseSphericalLimits=true
    gravityScale=0.35
    simulationSpace=BoneRelative
    alphaBoolBlend {blendOption=Linear}
    boundBone {boneName=dyn_backCloth_03}
    chainEnd {boneName=dyn_backCloth_04}
    componentAppliedLinearAccClamp=<6000 6000 6000>
    componentLinearAccScale=<1 1 0.3>
    componentLinearVelScale=<1 1 0.3>
    gravityOverride {z=-980}
    physicsBodyDefinitions [
     {
      boundBone {boneName=dyn_backCloth_03}
      boxExtents=<20 30 30>
      constraintSetup {angularTargetAxis=AxisY, angularLimitsMax=<15 15 0>, angularLimitsMin=<-15 -15 0>, angularTarget=<1 0 0>}
      localJointOffset=<2 0 0>
     }
     {
      boundBone {boneName=dyn_backCloth_04}
      boxExtents=<20 30 30>
      constraintSetup {angularTargetAxis=AxisY, angularLimitsMax=<15 15 15>, angularLimitsMin=<-15 -15 -15>, angularTarget=<1 0 0>}
      localJointOffset=<2 0 0>
     }
    ]
    relativeSpaceBone {boneName=spine_04}
    sphericalLimits [{limitRadius=0, drivingBone {boneName=None}, sphereLocalOffset=<0 0 0>}]
   }
   n1 = "Transform (Modify) Bone - Bone: dyn_loinCloth_MSTR"!(ComponentPose, bAlphaBoolEnabled:bool=True, Translation="0,0,0", Rotation="0,0,0", Scale="1,1,1") -> (Pose) {
    alphaInputType=Bool
    rotationMode=BMM_Additive
    rotationSpace=BCS_BoneSpace
    scaleMode=BMM_Additive
    scaleSpace=BCS_BoneSpace
    translationMode=BMM_Additive
    translationSpace=BCS_BoneSpace
    alphaBoolBlend {blendInTime=0.2, blendOption=Linear, blendOutTime=0.4}
    boneToModify {boneName=dyn_loinCloth_MSTR}
   }
   "Transform (Modify) Bone - Bone: dyn_loinCloth_MSTR"(ComponentPose=@n1, bAlphaBoolEnabled=True, Translation="0,0,0", Rotation="0,0,0", Scale="1,1,1") {
    alphaInputType=Bool
    rotationMode=BMM_Additive
    rotationSpace=BCS_BoneSpace
    scaleMode=BMM_Additive
    scaleSpace=BCS_BoneSpace
    translationMode=BMM_Additive
    translationSpace=BCS_BoneSpace
    alphaBoolBlend {blendInTime=0.2, blendOption=Linear, blendOutTime=0.4}
    boneToModify {boneName=dyn_loinCloth_MSTR}
   }
   n2 = "Transform (Modify) Bone - Bone: dyn_loinCloth_MSTR"(
    ComponentPose=Local To Component(LocalPose=Use cached pose 'shoulderPadCache'())
    bAlphaBoolEnabled=True
    Translation="-8,5,-2"
    Rotation="0,0,0"
    Scale="1,1,1"
   ) {
    alphaInputType=Bool
    rotationMode=BMM_Additive
    rotationSpace=BCS_BoneSpace
    scaleMode=BMM_Additive
    scaleSpace=BCS_BoneSpace
    translationMode=BMM_Additive
    translationSpace=BCS_BoneSpace
    alphaBoolBlend {blendInTime=5, blendOption=Linear, blendOutTime=3}
    boneToModify {boneName=dyn_loinCloth_MSTR}
    translation=<-8 5 -2>
   }
   n3 = "Transform (Modify) Bone - Bone: dyn_loinCloth_01"!(ComponentPose=@n2, bAlphaBoolEnabled:bool=True, Translation="0,0,0", Rotation="0,90,0", Scale="1,1,1") -> (Pose) {
    alphaInputType=Bool
    rotationMode=BMM_Additive
    rotationSpace=BCS_BoneSpace
    scaleMode=BMM_Additive
    scaleSpace=BCS_BoneSpace
    translationMode=BMM_Additive
    translationSpace=BCS_BoneSpace
    alphaBoolBlend {blendInTime=1, blendOption=Linear, blendOutTime=1}
    boneToModify {boneName=dyn_loinCloth_01}
    rotation {yaw=90}
   }
   n4 = Get isSprintingLoinCloth() -> (isSprintingLoinCloth:Vector)
   n5 = RigidBody!(
    ComponentPose=@n3
    Alpha
    OverrideWorldGravity:Vector="0,0,0"
    ExternalForce
    ComponentLinearAccScale:Vector=@n4
    ComponentLinearVelScale:Vector=@n4
   ) -> (Pose) {
    bClampLinearTranslationLimitToRefPose=true
    bOverrideWorldGravity=true
    overridePhysicsAsset=&/M_MED_FolkEvening_ButtCloth_Physics
    simulationSpace=BaseBoneSpace
    alphaBoolBlend {blendOption=Linear}
    baseBoneRef {boneName=pelvis}
    componentAppliedLinearAccClamp=<6000 6000 6000>
    componentLinearAccScale=<0.35 0.5 0.5>
    componentLinearVelScale=<0.35 0.5 0.5>
   }
   n6 = Component To Local(ComponentPose=@n5)
   n7 = "Transform (Modify) Bone - Bone: dyn_loinCloth_MSTR"!(
    ComponentPose=Local To Component(LocalPose=Use cached pose 'shoulderPadCache'())
    Alpha:float=1
    Translation="-8,5,-2"
    Rotation="1,0,0"
    Scale="1,1,1"
   ) -> (Pose) {
    rotationMode=BMM_Additive
    rotationSpace=BCS_BoneSpace
    scaleMode=BMM_Additive
    scaleSpace=BCS_BoneSpace
    translationMode=BMM_Additive
    translationSpace=BCS_BoneSpace
    alphaBoolBlend {blendOption=Linear}
    boneToModify {boneName=dyn_loinCloth_MSTR}
    rotation {pitch=1}
    translation=<-8 5 -2>
   }
   n8 = "Transform (Modify) Bone - Bone: dyn_belt_c_01"(ComponentPose=@n7, Alpha=1, Translation="0,0,0", Rotation="0,0,0", Scale="0.8,1,0.8") {
    rotationMode=BMM_Additive
    rotationSpace=BCS_BoneSpace
    scaleMode=BMM_Additive
    scaleSpace=BCS_BoneSpace
    translationMode=BMM_Additive
    translationSpace=BCS_BoneSpace
    alphaBoolBlend {blendOption=Linear}
    boneToModify {boneName=dyn_belt_c_01}
    scale {x=0.8, z=0.8}
   }
   n9 = "Transform (Modify) Bone - Bone: dyn_loinCloth_01"!(ComponentPose=@n8, Alpha:float=1, Translation="-4,0,0", Rotation="0,0,0", Scale="1,1,1") -> (Pose) {
    scaleMode=BMM_Additive
    scaleSpace=BCS_BoneSpace
    translationMode=BMM_Additive
    translationSpace=BCS_BoneSpace
    alphaBoolBlend {blendOption=Linear}
    boneToModify {boneName=dyn_loinCloth_01}
    translation {x=-4}
   }
   n10 = "Transform (Modify) Bone - Bone: dyn_loinCloth_01"!(ComponentPose=@n9, bAlphaBoolEnabled:bool=True, Translation="0,-10,0", Rotation="0,110,0", Scale="1,1,1") -> (Pose) {
    alphaInputType=Bool
    rotationMode=BMM_Additive
    rotationSpace=BCS_BoneSpace
    scaleMode=BMM_Additive
    scaleSpace=BCS_BoneSpace
    translationMode=BMM_Additive
    translationSpace=BCS_BoneSpace
    alphaBoolBlend {blendInTime=0.2, blendOption=Linear, blendOutTime=0.4}
    boneToModify {boneName=dyn_loinCloth_01}
    rotation {yaw=110}
    translation {y=-10}
   }
   n11 = RigidBody!(ComponentPose=@n10, Alpha, OverrideWorldGravity="0,0,0", ExternalForce) -> (Pose) {
    bClampLinearTranslationLimitToRefPose=true
    bOverrideWorldGravity=true
    overridePhysicsAsset=&/M_MED_FolkEvening_ButtCloth_isCrouching_Physics
    simulationSpace=BaseBoneSpace
    alphaBoolBlend {blendOption=Linear}
    baseBoneRef {boneName=pelvis}
    componentAppliedLinearAccClamp=<6000 6000 6000>
    componentLinearAccScale=<0.5 0.5 0.5>
    componentLinearVelScale=<0.5 0.5 0.5>
   }
   n12 = Component To Local(ComponentPose=@n11)
   n13 = "Transform (Modify) Bone - Bone: dyn_backLoinCloth_01"(
    ComponentPose=Local To Component(LocalPose=Blend Poses by bool(BlendPose_0=@n12, BlendPose_1=@n6, BlendTime_0=0, BlendTime_1=0) {blendType=Linear})
    Alpha=1
    Translation="0,0,0"
    Rotation="0,0,0"
    Scale="1,1,0.8"
   ) {
    scaleMode=BMM_Additive
    scaleSpace=BCS_BoneSpace
    alphaBoolBlend {blendOption=Linear}
    alphaScaleBiasClamp {bInterpResult=true, interpSpeedDecreasing=5}
    boneToModify {boneName=dyn_backLoinCloth_01}
    scale {z=0.8}
   }
   n14 = "Transform (Modify) Bone - Bone: dyn_shoulderPadTop_r_01"(ComponentPose=@n13, Alpha=1, Translation="-3,0,-2", Rotation="0,0,0", Scale="1,1,0.8") {
    scaleMode=BMM_Additive
    scaleSpace=BCS_BoneSpace
    translationMode=BMM_Additive
    translationSpace=BCS_BoneSpace
    alphaBoolBlend {blendOption=Linear}
    alphaScaleBiasClamp {bInterpResult=true, interpSpeedDecreasing=5}
    boneToModify {boneName=dyn_shoulderPadTop_r_01}
    scale {z=0.8}
    translation {x=-3, z=-2}
   }
   Output Pose(Result=Component To Local(ComponentPose=@n14)) {layerGroup=None}
   n15 = Control Rig!(Source=Copy Pose From Mesh() {bUseAttachedParent=true}, Alpha, isCrouching:bool=Get bIsCrouching()) -> (Pose) {
    alphaBoolBlend {blendOption=Linear}
    controlRigAssetReference {blueprintRigClass=&/M_MED_FolkEvening_ControlRig.M_MED_FolkEvening_ControlRig_C}
   }
   Save cached pose 'CRCache'(Pose=@n15) {cacheName=CRCache}
   n16 = "Constraint - dyn_backCloth_01 (P)"(ComponentPose=Local To Component(LocalPose=Use cached pose 'CRCache'()), Alpha=1, ConstraintWeights_0=1) {
    alphaBoolBlend {blendOption=Linear}
    boneToModify {boneName=dyn_backCloth_01}
    constraintSetup [{targetBone {boneName=spine_05}}]
    constraintWeights [1]
   }
   n17 = "Constraint - dyn_backCloth_02 (P)"(ComponentPose=@n16, Alpha=1, ConstraintWeights_0=1) {
    alphaBoolBlend {blendOption=Linear}
    boneToModify {boneName=dyn_backCloth_02}
    constraintSetup [{targetBone {boneName=spine_04}}]
    constraintWeights [1]
   }
   n18 = "Constraint - dyn_backCloth_03 (PP)"!(ComponentPose=@n17, Alpha=1, ConstraintWeights_0=0.5, ConstraintWeights_1:float=0.5) -> (Pose) {
    alphaBoolBlend {blendOption=Linear}
    boneToModify {boneName=dyn_backCloth_03}
    constraintSetup [{targetBone {boneName=spine_03}}, {targetBone {boneName=spine_02}}]
    constraintWeights [1,1]
   }
   n19 = "Constraint - dyn_backCloth_04 (P)"(ComponentPose=@n18, Alpha=1, ConstraintWeights_0=1) {
    alphaBoolBlend {blendOption=Linear}
    boneToModify {boneName=dyn_backCloth_04}
    constraintSetup [{targetBone {boneName=spine_01}}]
    constraintWeights [1]
   }
   n20 = "Transform (Modify) Bone - Bone: dyn_backCloth_03"!(ComponentPose=@n19, bAlphaBoolEnabled:bool=True, Translation="0,0,0", Rotation="0,-20,0", Scale="1,1,1") -> (Pose) {
    alphaInputType=Bool
    rotationMode=BMM_Additive
    rotationSpace=BCS_BoneSpace
    alphaBoolBlend {blendOption=Linear}
    boneToModify {boneName=dyn_backCloth_03}
    rotation {yaw=-20}
   }
   n21 = "Transform (Modify) Bone - Bone: dyn_backCloth_04"!(ComponentPose=@n20, bAlphaBoolEnabled:bool=True, Translation="0,0,0", Rotation="0,-15,0", Scale="1,1,1") -> (Pose) {
    alphaInputType=Bool
    rotationMode=BMM_Additive
    rotationSpace=BCS_BoneSpace
    alphaBoolBlend {blendOption=Linear}
    boneToModify {boneName=dyn_backCloth_04}
    rotation {yaw=-15}
   }
   n22 = Get isTacticalSprintingValsBackCloth() -> (isTacticalSprintingValsBackCloth:Vector)
   n23 = RigidBody!(
    ComponentPose=@n21
    Alpha
    OverrideWorldGravity="0,0,0"
    ExternalForce
    ComponentLinearAccScale:Vector=@n22
    ComponentLinearVelScale:Vector=@n22
   ) -> (Pose) {
    bClampLinearTranslationLimitToRefPose=true
    bOverrideWorldGravity=true
    overridePhysicsAsset=&/M_MED_FolkEvening_BackCloth_Backbling_Physics
    simulationSpace=BaseBoneSpace
    alphaBoolBlend {blendOption=Linear}
    baseBoneRef {boneName=spine_04}
    componentAppliedLinearAccClamp=<6000 6000 6000>
    componentLinearAccScale=<0.5 0.5 0.5>
    componentLinearVelScale=<0.5 0.5 0.5>
   }
   n24 = Get isTacticalSprintingValsBackCloth()
   n25 = RigidBody(
    ComponentPose=Local To Component(LocalPose=Use cached pose 'CRCache'())
    OverrideWorldGravity="0,0,0"
    ComponentLinearAccScale=@n24
    ComponentLinearVelScale=@n24
   ) {
    bClampLinearTranslationLimitToRefPose=true
    bOverrideWorldGravity=true
    overridePhysicsAsset=&/M_MED_FolkEvening_BackCloth_Physics
    simulationSpace=BaseBoneSpace
    alphaBoolBlend {blendOption=Linear}
    baseBoneRef {boneName=spine_04}
    componentAppliedLinearAccClamp=<6000 6000 6000>
    componentLinearAccScale=<0.5 0.5 0.5>
    componentLinearVelScale=<0.5 0.5 0.5>
   }
   Save cached pose 'backClothCache'(Pose=Blend Poses by bool(
     BlendPose_0=Component To Local(ComponentPose=@n23)
     BlendPose_1=Component To Local(ComponentPose=@n25)
     BlendTime_0=0
     BlendTime_1=0
     bActiveValue=Get Is Wearing Backbling()
    ) {blendType=Linear}) {cacheName=backClothCache}
   n26 = "Transform (Modify) Bone - Bone: dyn_shoulderPadTop_l_01"!(
    ComponentPose=Local To Component(LocalPose=Use cached pose 'backClothCache'())
    bAlphaBoolEnabled:bool=True
    Translation="0,0,0"
    Rotation="0,0,-10"
    Scale="1,1,1"
   ) -> (Pose) {
    alphaInputType=Bool
    rotationMode=BMM_Additive
    rotationSpace=BCS_BoneSpace
    scaleMode=BMM_Additive
    scaleSpace=BCS_BoneSpace
    translationMode=BMM_Additive
    translationSpace=BCS_BoneSpace
    alphaBoolBlend {blendOption=Linear}
    boneToModify {boneName=dyn_shoulderPadTop_l_01}
    rotation {roll=-10}
   }
   n27 = "Transform (Modify) Bone - Bone: dyn_shoulderPadBot_l_01"!(ComponentPose=@n26, bAlphaBoolEnabled:bool=True, Translation="0,0,0", Rotation="0,0,-10", Scale="1,1,1") -> (Pose) {
    alphaInputType=Bool
    rotationMode=BMM_Additive
    rotationSpace=BCS_BoneSpace
    scaleMode=BMM_Additive
    scaleSpace=BCS_BoneSpace
    translationMode=BMM_Additive
    translationSpace=BCS_BoneSpace
    alphaBoolBlend {blendOption=Linear}
    boneToModify {boneName=dyn_shoulderPadBot_l_01}
    rotation {roll=-10}
   }
   n28 = "Transform (Modify) Bone - Bone: dyn_shoulderPadTop_r_01"!(ComponentPose=@n27, bAlphaBoolEnabled:bool=True, Translation="0,0,0", Rotation="0,0,-10", Scale="1,1,1") -> (Pose) {
    alphaInputType=Bool
    rotationMode=BMM_Additive
    rotationSpace=BCS_BoneSpace
    scaleMode=BMM_Additive
    scaleSpace=BCS_BoneSpace
    translationMode=BMM_Additive
    translationSpace=BCS_BoneSpace
    alphaBoolBlend {blendOption=Linear}
    boneToModify {boneName=dyn_shoulderPadTop_r_01}
    rotation {roll=-10}
   }
   n29 = "Transform (Modify) Bone - Bone: dyn_shoulderPadBot_r_01"!(ComponentPose=@n28, bAlphaBoolEnabled:bool=True, Translation="0,0,0", Rotation="0,0,-30", Scale="1,1,1") -> (Pose) {
    alphaInputType=Bool
    rotationMode=BMM_Additive
    rotationSpace=BCS_BoneSpace
    scaleMode=BMM_Additive
    scaleSpace=BCS_BoneSpace
    translationMode=BMM_Additive
    translationSpace=BCS_BoneSpace
    alphaBoolBlend {blendOption=Linear}
    boneToModify {boneName=dyn_shoulderPadBot_r_01}
    rotation {roll=-30}
   }
   n30 = "Transform (Modify) Bone - Bone: dyn_shoulderPadTop_l_01"(ComponentPose=@n29, bAlphaBoolEnabled=True, Translation="8,0,0", Rotation="-10,0,-10", Scale="1,1,1") {
    alphaInputType=Bool
    rotationMode=BMM_Additive
    rotationSpace=BCS_BoneSpace
    scaleMode=BMM_Additive
    scaleSpace=BCS_BoneSpace
    translationMode=BMM_Additive
    translationSpace=BCS_BoneSpace
    alphaBoolBlend {blendOption=Linear}
    boneToModify {boneName=dyn_shoulderPadTop_l_01}
    rotation {pitch=-10, roll=-10}
    translation {x=8}
   }
   n31 = "Transform (Modify) Bone - Bone: dyn_shoulderPadBot_l_01"(ComponentPose=@n30, bAlphaBoolEnabled=True, Translation="10,0,0", Rotation="0,0,0", Scale="1,1,1") {
    alphaInputType=Bool
    rotationMode=BMM_Additive
    rotationSpace=BCS_BoneSpace
    scaleMode=BMM_Additive
    scaleSpace=BCS_BoneSpace
    translationMode=BMM_Additive
    translationSpace=BCS_BoneSpace
    alphaBoolBlend {blendOption=Linear}
    boneToModify {boneName=dyn_shoulderPadBot_l_01}
    translation {x=10}
   }
   n32 = "Transform (Modify) Bone - Bone: dyn_shoulderPadTop_r_01"(ComponentPose=@n31, bAlphaBoolEnabled=True, Translation="-8,0,0", Rotation="-10,0,-10", Scale="1,1,1") {
    alphaInputType=Bool
    rotationMode=BMM_Additive
    rotationSpace=BCS_BoneSpace
    scaleMode=BMM_Additive
    scaleSpace=BCS_BoneSpace
    translationMode=BMM_Additive
    translationSpace=BCS_BoneSpace
    alphaBoolBlend {blendOption=Linear}
    boneToModify {boneName=dyn_shoulderPadTop_r_01}
    rotation {pitch=-10, roll=-10}
    translation {x=-8}
   }
   n33 = "Transform (Modify) Bone - Bone: dyn_shoulderPadBot_r_01"(ComponentPose=@n32, bAlphaBoolEnabled=True, Translation="-10,0,0", Rotation="0,0,-30", Scale="1,1,1") {
    alphaInputType=Bool
    rotationMode=BMM_Additive
    rotationSpace=BCS_BoneSpace
    scaleMode=BMM_Additive
    scaleSpace=BCS_BoneSpace
    translationMode=BMM_Additive
    translationSpace=BCS_BoneSpace
    alphaBoolBlend {blendOption=Linear}
    boneToModify {boneName=dyn_shoulderPadBot_r_01}
    rotation {roll=-30}
    translation {x=-10}
   }
   n34 = "Transform (Modify) Bone - Bone: dyn_shoulderPadTop_r_01"(ComponentPose=@n33, bAlphaBoolEnabled=True, Translation="0,0,-2", Rotation="8,0,0", Scale="0.9,0.9,0.9") {
    alphaInputType=Bool
    rotationMode=BMM_Additive
    rotationSpace=BCS_BoneSpace
    scaleMode=BMM_Additive
    scaleSpace=BCS_BoneSpace
    translationMode=BMM_Additive
    translationSpace=BCS_BoneSpace
    alphaBoolBlend {blendOption=Linear}
    boneToModify {boneName=dyn_shoulderPadTop_r_01}
    rotation {pitch=8}
    scale=<0.9 0.9 0.9>
    translation {z=-2}
   }
   n35 = "Transform (Modify) Bone - Bone: dyn_shoulderPadTop_l_01"(ComponentPose=@n34, bAlphaBoolEnabled=True, Translation="0,0,0", Rotation="8,0,0", Scale="0.9,0.9,0.9") {
    alphaInputType=Bool
    rotationMode=BMM_Additive
    rotationSpace=BCS_BoneSpace
    scaleMode=BMM_Additive
    scaleSpace=BCS_BoneSpace
    translationMode=BMM_Additive
    translationSpace=BCS_BoneSpace
    alphaBoolBlend {blendOption=Linear}
    boneToModify {boneName=dyn_shoulderPadTop_l_01}
    rotation {pitch=8}
    scale=<0.9 0.9 0.9>
   }
   Save cached pose 'shoulderPadCache'(Pose=Component To Local(ComponentPose=@n35)) {cacheName=shoulderPadCache}
  ]}
 EventGraph {nodes [
   n1 = Event BlueprintUpdateAnimation()
   > Set Delta Time X(Delta Time X=@n1.DeltaTimeX)
   > n2 = Sequence!(execute) -> (then_0, then_1, then_2:Exec, then_3:Exec, then_4:Exec, then_5:Exec, then_6:Exec, then_7:Exec, then_8:Exec, then_9:Exec)
   > n3 = Gravity Override State Function DEPRECATED(CharacterPartAnimInstance=Self-Reference(), GameTimeInSeconds=0, GravityOverrideParams=Get Gravity Override Params())
   > Set Gravity Override(Gravity Override=@n3)
   n4 = PlayerMovement(execute=@n2.then_1, AnimInstance=Self-Reference(), XDotProductCoefficient=1, YDotProductCoefficient=1, ZDotProductCoefficient=1)
   > n5 = Branch(Condition=OR Boolean(A="Not Equal (Float)"(A=@n4), B="Not Equal (Float)"(A=@n4.YMovement)))
   > Set PlayerMovement(PlayerMovement:bool=true) -> (Output_Get:bool)
   Set PlayerMovement(execute=@n5.else, PlayerMovement=false)
   Set WPO Globals Function(
    execute=@n2.then_7
    Material=Get Mat Instance Dynamic()
    Skeletal Mesh=Get Skel Mesh()
    "Local Noise Transform  Location"=Get Local Noise Transform Location() -> (Local Noise Transform Location:Vector)
    Local Noise Offset=Get Local Noise Offset()
   )
   n6 = Get Delta Time X()
   > Set WPO Regional Function!(
    execute
    Material:Material Instance Dynamic Object Reference=Get Mat Instance Dynamic()
    Overlay Material:Material Instance Dynamic Object Reference
    Skeletal Mesh:Skeletal Mesh Component Object Reference=Get Skel Mesh()
    Velocity Transform:Name=pelvis
    Velocity Interp Speed:double=Get Velocity Interp Speed()
    Collision Transform:Name=pelvis
    Collision Transform Axis:EAxis Enum
    Invert Collision Transform Axis:bool=false
    Delta Time X:double=@n6
    Previous Velocity Transform Location:"Vector (by ref)"=Get R2 Velocity Transform Location()
    Previous Velocity Transform Velocity:"Vector (by ref)"=Get R2 Velocity Transform Velocity()
    Region:int=2
   ) -> (then)
   > n7 = Branch(Condition=Get Is Wearing Backbling())
   Set WPO Regional Function(
    execute=@n7.else
    Material=Get Mat Instance Dynamic()
    Skeletal Mesh=Get Skel Mesh()
    Velocity Transform=spine_05
    Velocity Interp Speed=Get Velocity Interp Speed()
    Collision Transform=spine_05
    Collision Transform Axis=Y
    Invert Collision Transform Axis=true
    Delta Time X=@n6
    Previous Velocity Transform Location=Get R1 Velocity Transform Location()
    Previous Velocity Transform Velocity=Get R1 Velocity Transform Velocity()
    Region=1
   )
   n8 = Branch(execute=@n2.then_4, Condition=AND Boolean(A=Get bIsCrouching(), B=Get PlayerMovement() -> (PlayerMovement:bool)))
   > Set isCrouchMoving(isCrouchMoving:bool=true) -> (Output_Get:bool)
   Set isCrouchMoving(execute=@n8.else, isCrouchMoving=false)
   n9 = Branch(execute=@n2.then_3, Condition="float < float"(A=@n4.YMovement))
   > Set isMovingBackward(isMovingBackward:bool=true) -> (Output_Get:bool)
   Set isMovingBackward(execute=@n9.else, isMovingBackward=false)
   n10 = Branch(execute=@n2.then_2, Condition="float > float"(A=@n4.YMovement))
   > Set isMovingForward(isMovingForward:bool=true) -> (Output_Get:bool)
   Set isMovingForward(execute=@n10.else, isMovingForward=false)
   n11 = CheckForBackbling(self=Self-Reference())
   n12 = Branch(execute=@n2.then_6, Condition=OR Boolean!(A=@n11, B=@n11.bIsBackblingACape, C:bool=@n11.bIsBackblingLong) -> (ReturnValue))
   > Set Is Wearing Backbling(Is Wearing Backbling=true)
   Set Is Wearing Backbling(execute=@n12.else, Is Wearing Backbling=false)
   n13 = Branch(execute=@n2.then_5, Condition=Get bIsTacticalSprinting())
   > Set isTacticalSprintingValsBackCloth(isTacticalSprintingValsBackCloth:Vector="0.001,0.001,0.001") -> (Output_Get:Vector)
   Set isTacticalSprintingValsBackCloth(execute=@n13.else, isTacticalSprintingValsBackCloth="0.5,0.5,0.5")
   n14 = Branch(execute=@n2.then_8, Condition=OR Boolean!(A=Get bIsSprinting(), B=Get bIsTacticalSprinting()) -> (ReturnValue))
   > Set isRunningLikeTheWind(isRunningLikeTheWind:bool=true) -> (Output_Get:bool)
   > Set isSprintingLoinCloth(isSprintingLoinCloth:Vector="0.5,0.5,0.5") -> (Output_Get:Vector)
   Set isRunningLikeTheWind(execute=@n14.else, isRunningLikeTheWind=false)
   > Set isSprintingLoinCloth(isSprintingLoinCloth="0.6,0.7,0.7")
   n15 = Cast To FortPlayerAnimInstance(execute=@n2.then_9, Object=GetAnimInstance(self=GetOwnerSkeletalMeshComponent()))
   > Set isWeaponMelee(isWeaponMelee:bool=false) -> (Output_Get:bool)
   > Switch on EFortWeaponCoreAnimation!(execute, Selection:EFortWeaponCoreAnimation Enum=Get WeaponCoreAnim(self=@n15)) -> (
    Melee:Exec
    Pistol:Exec
    Shotgun:Exec
    PaperBlueprint:Exec
    Rifle:Exec
    MeleeOneHand:Exec
    MachinePistol:Exec
    RocketLauncher:Exec
    GrenadeLauncher:Exec
    GoingCommando:Exec
    AssaultRifle:Exec
    TacticalShotgun:Exec
    SniperRifle:Exec
    TrapPlacement:Exec
    ShoulderLauncher:Exec
    AbilityDecoTool:Exec
    Crossbow:Exec
    C4:Exec
    RemoteControl:Exec
    DualWield:Exec
    AR_BullPup:Exec
    AR_ForwardGrip:Exec
    MedPackPaddles:Exec
    SMG_P90:Exec
    AR_DrumGun:Exec
    Consumable_Small:Exec
    Consumable_Large:Exec
    Balloon:Exec
    MountedTurret:Exec
    CreativeTool:Exec
    ExplosiveBow:Exec
    AshtonIndigo:Exec
    AshtonChicago:Exec
    MeleeDualWield:Exec
    Unarmed:Exec
   )
   > n16 = Branch()
   > Set isWeaponMelee(isWeaponMelee=true)
   Set isWeaponMelee(execute=@n16.else, isWeaponMelee=false)
   Event OnCharacterCustomizationLoaded()
   > Sequence!(execute) -> (then_0, then_1)
   > n17 = Set Skel Mesh(Skel Mesh=GetPartSkeletalMeshComponent())
   > n18 = Get Dynamic Material Instance(SkeletalMeshComponent=@n17)
   > Set Mat Instance Dynamic(Mat Instance Dynamic=@n18)
   Get Skel Mesh()
  ]}
]
variables [
 Delta Time X {type=number}
 Gravity Override {type=Vector}
 Gravity Override Params {type=GravityOverrideParamsStruct}
 Is Wearing Backbling {type=boolean}
 isCharMoving {type=boolean}
 isCrouchMoving {type=boolean}
 isMovingBackward {type=boolean}
 isMovingForward {type=boolean}
 isRunningLikeTheWind {type=boolean}
 isSprintingLoinCloth {type=Vector}
 isTacticalSprintingValsBackCloth {type=Vector}
 isWeaponMelee {type=boolean}
 Local Noise Offset {type=Vector}
 Local Noise Transform Location {type=Vector}
 Mat Instance Dynamic {type=MaterialInstanceDynamic}
 PlayerMovement {type=boolean}
 R1 Velocity Transform Location {type=Vector}
 R1 Velocity Transform Velocity {type=Vector}
 R2 Velocity Transform Location {type=Vector}
 R2 Velocity Transform Velocity {type=Vector}
 Skel Mesh {type=SkeletalMeshComponent}
 Velocity Interp Speed {type=number}
]
