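# Physics asset for the FolkEvening medium male character's backbling and face accessory mesh, setting up physics for beard and back-hair bones.
# Capsule bodies cover the pelvis, spine, clavicles, neck, head and jaw, with free-rotating joints except at the clavicles.
# Swing-limited, zero-twist chains form two back-hair strands off spine_05 and a main beard from a neck-parented master bone.
# Left and right beard chains branch from the main beard, and a shorter center chain hangs from the jaw.
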PhysicsAsset /BRCosmetics/Characters/Player/Male/Medium/Bodies/M_MED_FolkEvening/Meshes/Parts/M_MED_FolkEvening_Backbling_FaceAcc_Physics
bodies {
 C_jaw [{shapeType=Capsule, length=0.5, radius=0.5, shapeName=C_jaw_capsule}]
 clavicle_l [{shapeType=Capsule, length=10.5, radius=13.5, shapeName=clavicle_l_capsule, center=<8.5 0 0>, rotation=<rot 90 0 0>}]
 clavicle_r [{shapeType=Capsule, length=10.5, radius=13.5, shapeName=clavicle_l_capsule, center=<-8.5 0 0>, rotation=<rot 90 180 0>}]
 dyn_backHairB_03 [{shapeType=Capsule, length=2.2706, radius=4.4008, shapeName=dyn_backHairB_03_capsule, center=<3.0744 -2 0>, rotation=<rot 90 0 0>}]
 dyn_backHairB_04 [{shapeType=Capsule, length=1.6811, radius=5.2508, shapeName=dyn_backHairB_04_capsule, center=<3.0744 -2 0>, rotation=<rot 90 0 0>}]
 dyn_backHairF_03 [{shapeType=Capsule, length=7.5852, radius=2.2348, shapeName=dyn_backHairF_03_capsule, center=<3.0744 0 0>, rotation=<rot 90 0 0>}]
 dyn_backHairF_04 [{shapeType=Capsule, length=6.8753, radius=3.1247, shapeName=dyn_backHairF_04_capsule, center=<3.0744 0 0>, rotation=<rot 90 0 0>}]
 dyn_beardMain_c_01 [{shapeType=Capsule, length=4.3381, radius=1.327, shapeName=dyn_beardMain_c_01_capsule, center=<3.0744 1.5 0>, rotation=<rot 90 0 0>}]
 dyn_beardMain_c_02 [{shapeType=Capsule, length=4.3381, radius=1.327, shapeName=dyn_beardMain_c_02_capsule, center=<3.0744 1.5 0>, rotation=<rot 90 0 0>}]
 dyn_beardMain_c_03 [{shapeType=Capsule, length=4.3381, radius=1.327, shapeName=dyn_beardMain_c_03_capsule, center=<3.0744 1.5 0>, rotation=<rot 90 0 0>}]
 dyn_beardMain_c_04 [{shapeType=Capsule, length=2.8753, radius=2.7899, shapeName=dyn_beardMain_c_04_capsule, center=<3.3611 1.8 0>}]
 dyn_beardMain_c_05 [{shapeType=Capsule, length=2.8753, radius=2.7899, shapeName=dyn_beardMain_c_05_capsule, center=<3.6095 1.8 0>}]
 dyn_beardMain_c_06 [{shapeType=Capsule, length=2.7358, radius=2.9294, shapeName=dyn_beardMain_c_05_capsule, center=<3.6095 1.8 0>}]
 dyn_beardMain_c_07 [{shapeType=Capsule, length=1.9819, radius=3.6833, shapeName=dyn_beardMain_c_05_capsule, center=<3.6095 1.8 0>}]
 dyn_beardMain_c_MSTR [{shapeType=Capsule, length=0.5, radius=0.5, shapeName=dyn_beardMain_c_MSTR_capsule}]
 dyn_beardMeshB_l_01 [{shapeType=Capsule, length=4.3381, radius=1.327, shapeName=dyn_beardMeshB_l_01_capsule, center=<3.0744 1.5 0>, rotation=<rot 90 0 0>}]
 dyn_beardMeshB_l_02 [{shapeType=Capsule, length=4.3381, radius=1.327, shapeName=dyn_beardMeshB_l_02_capsule, center=<3.0744 1.5 0>, rotation=<rot 90 0 0>}]
 dyn_beardMeshB_l_03 [{shapeType=Capsule, length=4.3381, radius=1.327, shapeName=dyn_beardMeshB_l_03_capsule, center=<3.0744 1.5 0>, rotation=<rot 90 0 0>}]
 dyn_beardMeshB_r_01 [{shapeType=Capsule, length=4.3381, radius=1.327, shapeName=dyn_beardMeshB_r_01_capsule, center=<3.0744 1.5 0>, rotation=<rot 90 0 0>}]
 dyn_beardMeshB_r_02 [{shapeType=Capsule, length=4.3381, radius=1.327, shapeName=dyn_beardMeshB_r_02_capsule, center=<3.0744 1.5 0>, rotation=<rot 90 0 0>}]
 dyn_beardMeshB_r_03 [{shapeType=Capsule, length=4.3381, radius=1.327, shapeName=dyn_beardMeshB_r_03_capsule, center=<3.0744 1.5 0>, rotation=<rot 90 0 0>}]
 dyn_beardMesh_c_01 [{shapeType=Capsule, length=3.7, radius=1.9651, shapeName=dyn_beardMesh_c_01_capsule, center=<3.0744 0 0>, rotation=<rot 90 0 0>}]
 dyn_beardMesh_c_02 [{shapeType=Capsule, length=1.1597, radius=1.9651, shapeName=dyn_beardMesh_c_02_capsule, center=<3.0744 0 0>, rotation=<rot 90 0 0>}]
 dyn_beardMesh_c_03 [{shapeType=Capsule, length=3.5713, radius=2.0938, shapeName=dyn_beardMesh_c_03_capsule, center=<3.0744 0 0>, rotation=<rot 90 0 0>}]
 head [{shapeType=Capsule, length=5, radius=5, shapeName=head_capsule}]
 neck_02 [{shapeType=Capsule, length=5, radius=5, shapeName=neck_02_capsule}]
 pelvis [{shapeType=Capsule, length=9.3691, radius=15.158, shapeName=pelvis_capsule, rotation=<rot 0 -90 0>}]
 spine_01 [{shapeType=Capsule, length=10.4224, radius=14.1048, shapeName=pelvis_capsule, rotation=<rot 0 -90 0>}]
 spine_02 [{shapeType=Capsule, length=12.9472, radius=11.58, shapeName=pelvis_capsule, rotation=<rot 0 -90 0>}]
 spine_03 [{shapeType=Capsule, length=10.0607, radius=12.2485, shapeName=spine_03_capsule, center=<0 -0.7909 0>}]
 spine_04 [{shapeType=Capsule, length=5.522, radius=14.8204, shapeName=spine_04_capsule}]
 spine_05 [{shapeType=Capsule, length=8.671, radius=14.6234, shapeName=spine_05_capsule, center=<1.1372 -0.7229 0>}]
}
constraints [
 {bone1Name=dyn_backHairF_03, bone2Name=spine_05, swing1LimitDegrees=15, swing2LimitDegrees=15, twistLimitDegrees=0}
 {bone1Name=dyn_backHairF_04, bone2Name=dyn_backHairF_03, swing1LimitDegrees=10, swing2LimitDegrees=10, twistLimitDegrees=0}
 {bone1Name=dyn_backHairB_03, bone2Name=spine_05, swing1LimitDegrees=0, swing2LimitDegrees=20, twistLimitDegrees=0}
 {bone1Name=dyn_backHairB_04, bone2Name=dyn_backHairB_03, swing1LimitDegrees=10, swing2LimitDegrees=10, twistLimitDegrees=0}
 {
  bone1Name=neck_02
  bone2Name=spine_05
  swing1LimitDegrees=45
  swing1Motion=Free
  swing2LimitDegrees=45
  swing2Motion=Free
  twistLimitDegrees=45
  twistMotion=Free
 }
 {
  bone1Name=dyn_beardMain_c_MSTR
  bone2Name=neck_02
  swing1LimitDegrees=45
  swing1Motion=Free
  swing2LimitDegrees=45
  swing2Motion=Free
  twistLimitDegrees=45
  twistMotion=Free
 }
 {bone1Name=dyn_beardMain_c_01, bone2Name=dyn_beardMain_c_MSTR, swing1LimitDegrees=20, swing2LimitDegrees=8, twistLimitDegrees=0}
 {bone1Name=dyn_beardMain_c_02, bone2Name=dyn_beardMain_c_01, swing1LimitDegrees=10, swing2LimitDegrees=8, twistLimitDegrees=0}
 {bone1Name=dyn_beardMain_c_03, bone2Name=dyn_beardMain_c_02, swing1LimitDegrees=20, swing2LimitDegrees=20, twistLimitDegrees=0}
 {bone1Name=dyn_beardMain_c_04, bone2Name=dyn_beardMain_c_03, swing1LimitDegrees=30, swing2LimitDegrees=20, twistLimitDegrees=0}
 {bone1Name=dyn_beardMain_c_05, bone2Name=dyn_beardMain_c_04, swing1LimitDegrees=30, swing2LimitDegrees=20, twistLimitDegrees=0}
 {bone1Name=dyn_beardMeshB_l_01, bone2Name=dyn_beardMain_c_02, swing1LimitDegrees=8, swing2LimitDegrees=8, twistLimitDegrees=0}
 {bone1Name=dyn_beardMeshB_l_02, bone2Name=dyn_beardMeshB_l_01, swing1LimitDegrees=20, swing2LimitDegrees=20, twistLimitDegrees=0}
 {bone1Name=dyn_beardMeshB_l_03, bone2Name=dyn_beardMeshB_l_02, swing1LimitDegrees=20, swing2LimitDegrees=20, twistLimitDegrees=0}
 {bone1Name=dyn_beardMeshB_r_01, bone2Name=dyn_beardMain_c_02, swing1LimitDegrees=8, swing2LimitDegrees=8, twistLimitDegrees=0}
 {bone1Name=dyn_beardMeshB_r_02, bone2Name=dyn_beardMeshB_r_01, swing1LimitDegrees=20, swing2LimitDegrees=20, twistLimitDegrees=0}
 {bone1Name=dyn_beardMeshB_r_03, bone2Name=dyn_beardMeshB_r_02, swing1LimitDegrees=20, swing2LimitDegrees=20, twistLimitDegrees=0}
 {bone1Name=dyn_beardMesh_c_01, bone2Name=C_jaw, swing1LimitDegrees=30, swing2LimitDegrees=20, twistLimitDegrees=0}
 {bone1Name=dyn_beardMesh_c_02, bone2Name=dyn_beardMesh_c_01, swing1LimitDegrees=20, swing2LimitDegrees=20, twistLimitDegrees=0}
 {bone1Name=dyn_beardMesh_c_03, bone2Name=dyn_beardMesh_c_02, swing1LimitDegrees=20, swing2LimitDegrees=20, twistLimitDegrees=0}
 {
  bone1Name=head
  bone2Name=neck_02
  swing1LimitDegrees=45
  swing1Motion=Free
  swing2LimitDegrees=45
  swing2Motion=Free
  twistLimitDegrees=45
  twistMotion=Free
 }
 {bone1Name=clavicle_l, bone2Name=spine_05, swing1LimitDegrees=20, swing2LimitDegrees=20, twistLimitDegrees=0}
 {
  bone1Name=spine_05
  bone2Name=spine_04
  swing1LimitDegrees=45
  swing1Motion=Free
  swing2LimitDegrees=45
  swing2Motion=Free
  twistLimitDegrees=45
  twistMotion=Free
 }
 {
  bone1Name=spine_04
  bone2Name=spine_03
  swing1LimitDegrees=45
  swing1Motion=Free
  swing2LimitDegrees=45
  swing2Motion=Free
  twistLimitDegrees=45
  twistMotion=Free
 }
 {
  bone1Name=spine_03
  bone2Name=spine_02
  swing1LimitDegrees=45
  swing1Motion=Free
  swing2LimitDegrees=45
  swing2Motion=Free
  twistLimitDegrees=45
  twistMotion=Free
 }
 {
  bone1Name=spine_02
  bone2Name=spine_01
  swing1LimitDegrees=45
  swing1Motion=Free
  swing2LimitDegrees=45
  swing2Motion=Free
  twistLimitDegrees=45
  twistMotion=Free
 }
 {
  bone1Name=spine_01
  bone2Name=pelvis
  swing1LimitDegrees=45
  swing1Motion=Free
  swing2LimitDegrees=45
  swing2Motion=Free
  twistLimitDegrees=45
  twistMotion=Free
 }
 {bone1Name=dyn_beardMain_c_06, bone2Name=dyn_beardMain_c_05, swing1LimitDegrees=30, swing2LimitDegrees=20, twistLimitDegrees=0}
 {bone1Name=dyn_beardMain_c_07, bone2Name=dyn_beardMain_c_06, swing1LimitDegrees=30, swing2LimitDegrees=20, twistLimitDegrees=0}
 {bone1Name=clavicle_r, bone2Name=spine_05, swing1LimitDegrees=20, swing2LimitDegrees=20, twistLimitDegrees=0}
 {
  bone1Name=C_jaw
  bone2Name=neck_02
  swing1LimitDegrees=45
  swing1Motion=Free
  swing2LimitDegrees=45
  swing2Motion=Free
  twistLimitDegrees=45
  twistMotion=Free
 }
]
