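# Physics asset for the FolkEvening medium male face-accessory mesh, defining capsule bodies and joints for back hair, beard and back cloth.
# Capsules on pelvis, spine, neck, head, jaw, clavicles and upper arms form the body framework, joined mostly by free constraints.
# Hair branches in three chains from a neck-rooted main chain; beard chains hang from neck and jaw.
# Cloth hangs from the left clavicle; dynamic joints mostly use tight angular limits.
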PhysicsAsset /BRCosmetics/Characters/Player/Male/Medium/Bodies/M_MED_FolkEvening/Meshes/Parts/M_MED_FolkEvening_FaceAcc_Physics
bodies {
 C_jaw [{shapeType=Capsule, length=0.5, radius=0.5, shapeName=C_jaw_capsule}]
 clavicle_l [{shapeType=Capsule, length=10.5, radius=13.5, shapeName=clavicle_l_capsule, center=<8.5 0 0>, rotation=<rot 90 0 0>}]
 clavicle_r [{shapeType=Capsule, length=10.3561, radius=12.5192, shapeName=clavicle_l_capsule, center=<-8.5 0 0>, rotation=<rot 90 180 0>}]
 dyn_backCloth_01 [{shapeType=Capsule, length=13.6614, radius=14.6447, shapeName=dyn_backCloth_01_capsule, center=<5.8768 12.8972 0>, rotation=<rot 90 0 0>}]
 dyn_backCloth_02 [{shapeType=Capsule, length=13.6614, radius=14.6447, shapeName=dyn_backCloth_02_capsule, center=<5.8768 12.8972 0>, rotation=<rot 90 0 0>}]
 dyn_backCloth_03 [{shapeType=Capsule, length=13.6614, radius=14.6447, shapeName=dyn_backCloth_02_capsule, center=<5.8768 12.8972 0>, rotation=<rot 90 0 0>}]
 dyn_backHairB_01 [{shapeType=Capsule, length=6.0397, radius=3.9603, shapeName=dyn_backHairB_01_capsule, center=<3.0744 -2 0>, rotation=<rot 90 0 0>}]
 dyn_backHairB_02 [{shapeType=Capsule, length=5.4496, radius=4.0718, shapeName=dyn_backHairB_02_capsule, center=<3.0744 -2 0>, rotation=<rot 90 0 0>}]
 dyn_backHairB_03 [{shapeType=Capsule, length=2.8405, radius=3.8309, shapeName=dyn_backHairB_03_capsule, center=<3.0744 -2 0>, rotation=<rot 90 0 0>}]
 dyn_backHairB_04 [{shapeType=Capsule, length=5.0386, radius=2.9232, shapeName=dyn_backHairB_04_capsule, center=<3.0744 -2 0>, rotation=<rot 90 0 0>}]
 dyn_backHairD_01 [{shapeType=Capsule, length=2.3338, radius=3.2362, shapeName=dyn_backHairD_01_capsule, center=<4.5772 -1.7 0>, rotation=<rot 90 0 0>}]
 dyn_backHairD_02 [{shapeType=Capsule, length=2.5235, radius=3.0465, shapeName=dyn_backHairD_02_capsule, center=<4.5772 -1.7 0>, rotation=<rot 90 0 0>}]
 dyn_backHairD_03 [{shapeType=Capsule, length=3.7647, radius=3.2028, shapeName=dyn_backHairD_03_capsule, center=<4.5772 -1.7 0>, rotation=<rot 90 0 0>}]
 dyn_backHairF_01 [{shapeType=Capsule, length=7.9999, radius=2.0001, shapeName=dyn_backHairF_01_capsule, center=<3.0744 0 0>, rotation=<rot 90 0 0>}]
 dyn_backHairF_02 [{shapeType=Capsule, length=8.2199, radius=1.7801, shapeName=dyn_backHairF_02_capsule, center=<3.0744 0 0>, rotation=<rot 90 0 0>}]
 dyn_backHairF_03 [{shapeType=Capsule, length=8.35, radius=1.47, shapeName=dyn_backHairF_03_capsule, center=<3.0744 0 0>, rotation=<rot 90 0 0>}]
 dyn_backHairF_04 [{shapeType=Capsule, length=7.8092, radius=2.1908, shapeName=dyn_backHairF_04_capsule, center=<3.0744 0 0>, rotation=<rot 90 0 0>}]
 dyn_backHairMain_01 [{shapeType=Capsule, length=14.3561, radius=3.0868, shapeName=dyn_backHairMain_01_capsule, center=<3.2729 0 1.5>, rotation=<rot 90 180 -90>}]
 dyn_backHairMain_02 [{shapeType=Capsule, length=12.7682, radius=4.7664, shapeName=dyn_backHairMain_02_capsule, center=<4 0 3.9>, rotation=<rot -0.0001 0 86.4821>}]
 dyn_backHairMain_03 [{shapeType=Capsule, length=13.1496, radius=3.0446, shapeName=dyn_backHairMain_03_capsule, center=<3.0744 -2 0>, rotation=<rot 0 90 0>}]
 dyn_backHairMain_MSTR [{shapeType=Capsule, length=0.5, radius=0.5, shapeName=dyn_backHairMain_MSTR_capsule}]
 dyn_beardMain_c_01 [{shapeType=Capsule, length=4.3381, radius=1.327, shapeName=dyn_beardMain_c_01_capsule, center=<3.0744 1.5 0>, rotation=<rot 90 0 0>}]
 dyn_beardMain_c_02 [{shapeType=Capsule, length=4.3381, radius=1.327, shapeName=dyn_beardMain_c_02_capsule, center=<3.0744 1.5 0>, rotation=<rot 90 0 0>}]
 dyn_beardMain_c_03 [{shapeType=Capsule, length=4.3381, radius=1.327, shapeName=dyn_beardMain_c_03_capsule, center=<3.0744 1.5 0>, rotation=<rot 90 0 0>}]
 dyn_beardMain_c_04 [{shapeType=Capsule, length=3.0367, radius=3.4894, shapeName=dyn_beardMain_c_04_capsule, center=<4.2299 1.8 0>}]
 dyn_beardMain_c_05 [{shapeType=Capsule, length=12.4538, radius=3.4894, shapeName=dyn_beardMain_c_05_capsule, center=<8.8294 1.8 0>, rotation=<rot 90 0 0>}]
 dyn_beardMain_c_MSTR [{shapeType=Capsule, length=0.5, radius=0.5, shapeName=dyn_beardMain_c_MSTR_capsule}]
 dyn_beardMesh_c_01 [{shapeType=Capsule, length=3.6522, radius=2.0129, shapeName=dyn_beardMesh_c_01_capsule, center=<3.0744 0 0>, rotation=<rot 90 0 0>}]
 dyn_beardMesh_c_02 [{shapeType=Capsule, length=8.4461, radius=2.0604, shapeName=dyn_beardMesh_c_02_capsule, center=<3.0744 0 0>, rotation=<rot 90 0 0>}]
 head_alt [{shapeType=Capsule, length=5, radius=5, shapeName=head_capsule}]
 neck_01 [{shapeType=Capsule, length=4.1503, radius=10.2995, shapeName=neck_01_capsule, center=<0 -1.4768 0>}]
 neck_02 [{shapeType=Capsule, length=5, radius=5, shapeName=neck_02_capsule}]
 pelvis [{shapeType=Capsule, length=9.3691, radius=15.158, shapeName=pelvis_capsule, rotation=<rot 0 -90 0>}]
 spine_01 [{shapeType=Capsule, length=9.0541, radius=15.473, shapeName=pelvis_capsule, center=<0.2242 -1.3927 0>, rotation=<rot 0 90 0>}]
 spine_02 [{shapeType=Capsule, length=10.2364, radius=14.2907, shapeName=pelvis_capsule, center=<0.2242 -1.3381 0>, rotation=<rot 0 90 0>}]
 spine_03 [{shapeType=Capsule, length=8.3071, radius=14.0021, shapeName=spine_03_capsule, center=<0 -1.3961 0>, rotation=<rot 0 90 0>}]
 spine_04 [{shapeType=Capsule, length=4.9116, radius=15.4308, shapeName=spine_04_capsule, center=<0 -1.3053 0>}]
 spine_05 [
  {shapeType=Capsule, length=9.0428, radius=14.2517, shapeName=spine_05_capsule, center=<1.1372 -0.7229 0>}
  {shapeType=Capsule, length=14.2211, radius=12.771, shapeName=spine_05_capsule, center=<6.353 -1.8456 0>}
 ]
 upperarm_l [{shapeType=Capsule, length=31.5, radius=11.505, shapeName=upperarm_l_capsule, center=<11.0139 0 4.1684>, rotation=<rot 90 0 0>}]
 upperarm_r [{
   shapeType=Capsule
   length=32.777
   radius=13.4349
   shapeName=upperarm_r_capsule
   center=<-11.3528 0.0002 -6.4107>
   rotation=<rot 81.263 -179.9658 -179.9695>
  }]
}
constraints [
 {bone1Name=dyn_backHairMain_MSTR, bone2Name=neck_01, swing1LimitDegrees=20, swing2LimitDegrees=20, twistLimitDegrees=0}
 {bone1Name=dyn_backHairMain_01, bone2Name=dyn_backHairMain_MSTR, swing1LimitDegrees=8, swing2LimitDegrees=20, twistLimitDegrees=20}
 {bone1Name=dyn_backHairMain_02, bone2Name=dyn_backHairMain_01, swing1LimitDegrees=10, swing2LimitDegrees=30, twistLimitDegrees=20}
 {bone1Name=dyn_backHairF_01, bone2Name=dyn_backHairMain_02, swing1LimitDegrees=20, swing2LimitDegrees=20, twistLimitDegrees=0}
 {bone1Name=dyn_backHairF_02, bone2Name=dyn_backHairF_01, swing1LimitDegrees=20, swing2LimitDegrees=20, twistLimitDegrees=0}
 {bone1Name=dyn_backHairF_03, bone2Name=dyn_backHairF_02, swing1LimitDegrees=45, swing2LimitDegrees=20, twistLimitDegrees=0}
 {bone1Name=dyn_backHairF_04, bone2Name=dyn_backHairF_03, swing1LimitDegrees=45, swing2LimitDegrees=20, twistLimitDegrees=0}
 {bone1Name=dyn_backHairMain_03, bone2Name=dyn_backHairMain_02, swing1LimitDegrees=20, swing2LimitDegrees=20, twistLimitDegrees=10}
 {bone1Name=dyn_backHairB_01, bone2Name=dyn_backHairMain_03, swing1LimitDegrees=20, swing2LimitDegrees=20, twistLimitDegrees=20}
 {bone1Name=dyn_backHairB_02, bone2Name=dyn_backHairB_01, swing1LimitDegrees=45, swing2LimitDegrees=20, twistLimitDegrees=0}
 {bone1Name=dyn_backHairB_03, bone2Name=dyn_backHairB_02, swing1LimitDegrees=45, swing2LimitDegrees=20, twistLimitDegrees=0}
 {bone1Name=dyn_backHairB_04, bone2Name=dyn_backHairB_03, swing1LimitDegrees=45, swing2LimitDegrees=20, twistLimitDegrees=0}
 {bone1Name=dyn_backHairD_01, bone2Name=dyn_backHairMain_03, swing1LimitDegrees=20, swing2LimitDegrees=20, twistLimitDegrees=0}
 {bone1Name=dyn_backHairD_02, bone2Name=dyn_backHairD_01, swing1LimitDegrees=20, swing2LimitDegrees=20, twistLimitDegrees=0}
 {bone1Name=dyn_backHairD_03, bone2Name=dyn_backHairD_02, swing1LimitDegrees=45, swing2LimitDegrees=20, twistLimitDegrees=0}
 {
  bone1Name=neck_02
  bone2Name=neck_01
  swing1LimitDegrees=45
  swing1Motion=Free
  swing2LimitDegrees=45
  swing2Motion=Free
  twistLimitDegrees=45
  twistMotion=Free
 }
 {
  bone1Name=dyn_beardMain_c_MSTR
  bone2Name=neck_02
  swing1LimitDegrees=45
  swing1Motion=Free
  swing2LimitDegrees=45
  swing2Motion=Free
  twistLimitDegrees=45
  twistMotion=Free
 }
 {bone1Name=dyn_beardMain_c_01, bone2Name=dyn_beardMain_c_MSTR, swing1LimitDegrees=20, swing2LimitDegrees=8, twistLimitDegrees=0}
 {bone1Name=dyn_beardMain_c_02, bone2Name=dyn_beardMain_c_01, swing1LimitDegrees=10, swing2LimitDegrees=8, twistLimitDegrees=0}
 {bone1Name=dyn_beardMain_c_03, bone2Name=dyn_beardMain_c_02, swing1LimitDegrees=20, swing2LimitDegrees=20, twistLimitDegrees=0}
 {bone1Name=dyn_beardMain_c_04, bone2Name=dyn_beardMain_c_03, swing1LimitDegrees=30, swing2LimitDegrees=30, twistLimitDegrees=0}
 {bone1Name=dyn_beardMain_c_05, bone2Name=dyn_beardMain_c_04, swing1LimitDegrees=30, swing2LimitDegrees=30, twistLimitDegrees=0}
 {bone1Name=dyn_beardMesh_c_01, bone2Name=C_jaw, swing1LimitDegrees=30, swing2LimitDegrees=8, twistLimitDegrees=0}
 {bone1Name=dyn_beardMesh_c_02, bone2Name=dyn_beardMesh_c_01, swing1LimitDegrees=30, swing2LimitDegrees=8, twistLimitDegrees=0}
 {bone1Name=neck_01, bone2Name=spine_05, swing1LimitDegrees=20, swing2LimitDegrees=20, twistLimitDegrees=0}
 {bone1Name=clavicle_l, bone2Name=spine_05, swing1LimitDegrees=20, swing2LimitDegrees=20, twistLimitDegrees=0}
 {
  bone1Name=spine_05
  bone2Name=spine_04
  swing1LimitDegrees=45
  swing1Motion=Free
  swing2LimitDegrees=45
  swing2Motion=Free
  twistLimitDegrees=45
  twistMotion=Free
 }
 {
  bone1Name=spine_04
  bone2Name=spine_03
  swing1LimitDegrees=45
  swing1Motion=Free
  swing2LimitDegrees=45
  swing2Motion=Free
  twistLimitDegrees=45
  twistMotion=Free
 }
 {bone1Name=dyn_backCloth_02, bone2Name=dyn_backCloth_01, swing1LimitDegrees=20, swing2LimitDegrees=20, twistLimitDegrees=0}
 {bone1Name=dyn_backCloth_01, bone2Name=clavicle_l, swing1LimitDegrees=20, swing2LimitDegrees=20, twistLimitDegrees=0}
 {
  bone1Name=spine_03
  bone2Name=spine_02
  swing1LimitDegrees=45
  swing1Motion=Free
  swing2LimitDegrees=45
  swing2Motion=Free
  twistLimitDegrees=45
  twistMotion=Free
 }
 {
  bone1Name=upperarm_r
  bone2Name=clavicle_r
  swing1LimitDegrees=45
  swing1Motion=Free
  swing2LimitDegrees=45
  swing2Motion=Free
  twistLimitDegrees=45
  twistMotion=Free
 }
 {
  bone1Name=spine_02
  bone2Name=spine_01
  swing1LimitDegrees=45
  swing1Motion=Free
  swing2LimitDegrees=45
  swing2Motion=Free
  twistLimitDegrees=45
  twistMotion=Free
 }
 {
  bone1Name=spine_01
  bone2Name=pelvis
  swing1LimitDegrees=45
  swing1Motion=Free
  swing2LimitDegrees=45
  swing2Motion=Free
  twistLimitDegrees=45
  twistMotion=Free
 }
 {bone1Name=clavicle_r, bone2Name=spine_05, swing1LimitDegrees=20, swing2LimitDegrees=20, twistLimitDegrees=0}
 {
  bone1Name=upperarm_l
  bone2Name=clavicle_l
  swing1LimitDegrees=45
  swing1Motion=Free
  swing2LimitDegrees=45
  swing2Motion=Free
  twistLimitDegrees=45
  twistMotion=Free
 }
 {
  bone1Name=dyn_backCloth_03
  bone2Name=dyn_backCloth_02
  swing1LimitDegrees=45
  swing1Motion=Free
  swing2LimitDegrees=45
  swing2Motion=Free
  twistLimitDegrees=45
  twistMotion=Free
 }
 {
  bone1Name=head_alt
  bone2Name=neck_02
  swing1LimitDegrees=45
  swing1Motion=Free
  swing2LimitDegrees=45
  swing2Motion=Free
  twistLimitDegrees=45
  twistMotion=Free
 }
 {
  bone1Name=C_jaw
  bone2Name=head_alt
  swing1LimitDegrees=45
  swing1Motion=Free
  swing2LimitDegrees=45
  swing2Motion=Free
  twistLimitDegrees=45
  twistMotion=Free
 }
]
