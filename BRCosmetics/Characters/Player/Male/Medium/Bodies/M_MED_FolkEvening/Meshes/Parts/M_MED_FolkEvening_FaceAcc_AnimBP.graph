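# Animation Blueprint for the FolkEvening male face accessory part, adding secondary motion to back hair, beard and string bones.
# Poses copied from parent and body meshes feed a Control Rig, then one of two RigidBody setups chosen by backbling detection.
# The backbling branch reshapes back hair, partly by head tilt, while the other pins hair to the head in the front end outside emotes.
# AnimDynamics string motion and a beard-to-spine constraint finish the pose.
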AnimBlueprint /BRCosmetics/Characters/Player/Male/Medium/Bodies/M_MED_FolkEvening/Meshes/Parts/M_MED_FolkEvening_FaceAcc_AnimBP
target_skeleton=&/M_MED_FolkEvening_FaceAcc_Skeleton
graphs [
 AnimGraph {nodes [
   n1 = Anim Dynamics - Bone: dyn_string_r_01(ComponentPose=Local To Component(LocalPose=Use cached pose 'dynamicCache'()), Alpha=1, ExternalForce="0,0,0") {
    angularDampingOverride=1
    angularSpringConstant=20000
    bAngularSpring=true
    bOverrideAngularDamping=true
    bUseGravityOverride=true
    gravityScale=0.35
    simulationSpace=RootRelative
    alphaBoolBlend {blendOption=Linear}
    boundBone {boneName=dyn_string_r_01}
    componentLinearAccScale=<1.5 1.5 1.5>
    componentLinearVelScale=<1.5 1.5 1.5>
    gravityOverride {z=-980}
    physicsBodyDefinitions [{
      boundBone {boneName=dyn_string_r_01}
      boxExtents=<20 30 30>
      constraintSetup {angularLimitsMax=<15 15 15>, angularLimitsMin=<-15 -15 -15>, angularTarget=<1 0 0>}
      localJointOffset=<4 0 0>
     }]
   }
   n2 = "Constraint - dyn_beardMain_c_01 (P)"!(ComponentPose=@n1, bAlphaBoolEnabled:bool=True, ConstraintWeights_0=1) -> (Pose) {
    alphaInputType=Bool
    alphaBoolBlend {blendOption=Linear}
    boneToModify {boneName=dyn_beardMain_c_01}
    constraintSetup [{targetBone {boneName=spine_05}}]
    constraintWeights [1]
   }
   Output Pose(Result=Component To Local(ComponentPose=@n2)) {layerGroup=None}
   n3 = Layered blend per bone(
    BasePose=Copy Pose From Mesh() {bUseAttachedParent=true}
    BlendPoses_0=Copy Pose From Mesh(SourceMeshComponent=Get Body Mesh() -> (Body Mesh:Skeletal Mesh Component Object Reference)) {bUseAttachedParent=true}
   ) {layerSetup [{branchFilters [
       {boneName=dyn_backCloth_01}
       {boneName=dyn_backCloth_02}
       {boneName=dyn_backCloth_03}
       {boneName=dyn_backCloth_04}
       {boneName=dyn_backCloth_MSTR}
       {boneName=clavicle_r}
      ]}]}
   n4 = Control Rig(Source=@n3) {
    alphaBoolBlend {blendOption=Linear}
    controlRigAssetReference {blueprintRigClass=&/M_MED_FolkEvening_FaceAcc_CtrlRig.M_MED_FolkEvening_FaceAcc_CtrlRig_C}
   }
   Save cached pose 'mechanicsCache'(Pose=@n4) {cacheName=mechanicsCache}
   Save cached pose 'testCache'() {cacheName=testCache}
   n5 = "Constraint - dyn_backHairMain_MSTR (P)"!(
    ComponentPose=Local To Component(LocalPose=Use cached pose 'mechanicsCache'())
    bAlphaBoolEnabled:bool=Get FrontEndConstraintToggle() -> (FrontEndConstraintToggle:bool)
    ConstraintWeights_0=1
   ) -> (Pose) {
    alphaInputType=Bool
    alphaBoolBlend {blendOption=Linear}
    boneToModify {boneName=dyn_backHairMain_MSTR}
    constraintSetup [{targetBone {boneName=head}}]
    constraintWeights [1]
   }
   n6 = RigidBody!(ComponentPose=@n5, Alpha, OverrideWorldGravity:Vector=Get Gravity Override(), ExternalForce) -> (Pose) {
    bOverrideWorldGravity=true
    overridePhysicsAsset=&/M_MED_FolkEvening_FaceAcc_Physics
    simulationSpace=BaseBoneSpace
    alphaBoolBlend {blendOption=Linear}
    baseBoneRef {boneName=neck_01}
    componentAppliedLinearAccClamp=<6000 6000 6000>
    componentLinearAccScale=<0.4 0.5 0.1>
    componentLinearVelScale=<0.4 0.5 0.1>
   }
   n7 = RigidBody(ComponentPose=Local To Component(LocalPose=Use cached pose 'mechanicsCache'()), OverrideWorldGravity=Get Gravity Override()) {
    bOverrideWorldGravity=true
    overridePhysicsAsset=&/M_MED_FolkEvening_Backbling_FaceAcc_Physics
    simulationSpace=BaseBoneSpace
    alphaBoolBlend {blendOption=Linear}
    baseBoneRef {boneName=spine_05}
    componentAppliedLinearAccClamp=<6000 6000 6000>
    componentLinearAccScale=<0.6 0.5 0.5>
    componentLinearVelScale=<0.6 0.5 0.5>
   }
   n8 = "Constraint - dyn_backHairMain_02 (PP)"!(ComponentPose=@n7, Alpha=1, ConstraintWeights_0=1, ConstraintWeights_1:float=0) -> (Pose) {
    alphaBoolBlend {blendOption=Linear}
    boneToModify {boneName=dyn_backHairMain_02}
    constraintSetup [{targetBone {boneName=spine_05}}, {targetBone {boneName=neck_02}}]
    constraintWeights [1,1]
   }
   n9 = "Transform (Modify) Bone - Bone: dyn_backHairMain_02"(ComponentPose=@n8, Alpha=Get head rot back() -> (head rot back:float), Translation="3,-3,0", Rotation="0,0,0", Scale="1,1,1") {
    translationMode=BMM_Additive
    translationSpace=BCS_BoneSpace
    alphaBoolBlend {blendOption=Linear}
    alphaScaleBiasClamp {bInterpResult=true, interpSpeedDecreasing=5, interpSpeedIncreasing=5}
    boneToModify {boneName=dyn_backHairMain_02}
    translation=<3 -3>
   }
   n10 = "Transform (Modify) Bone - Bone: dyn_backHairF_03"!(ComponentPose=@n9, bAlphaBoolEnabled:bool=True, Translation="0,0,0", Rotation="0,-50,0", Scale="1,1,1") -> (Pose) {
    alphaInputType=Bool
    rotationMode=BMM_Additive
    rotationSpace=BCS_BoneSpace
    translationMode=BMM_Additive
    translationSpace=BCS_BoneSpace
    alphaBoolBlend {blendOption=Linear}
    alphaScaleBiasClamp {bInterpResult=true, interpSpeedDecreasing=5, interpSpeedIncreasing=5}
    boneToModify {boneName=dyn_backHairF_03}
    rotation {yaw=-50}
   }
   n11 = "Transform (Modify) Bone - Bone: dyn_backHairB_03"!(ComponentPose=@n10, bAlphaBoolEnabled:bool=True, Translation="0,0,0", Rotation="0,-50,0", Scale="1,1,1") -> (Pose) {
    alphaInputType=Bool
    rotationMode=BMM_Additive
    rotationSpace=BCS_BoneSpace
    translationMode=BMM_Additive
    translationSpace=BCS_BoneSpace
    alphaBoolBlend {blendOption=Linear}
    alphaScaleBiasClamp {bInterpResult=true, interpSpeedDecreasing=5, interpSpeedIncreasing=5}
    boneToModify {boneName=dyn_backHairB_03}
    rotation {yaw=-50}
   }
   n12 = "Transform (Modify) Bone - Bone: dyn_backHairB_02"(ComponentPose=@n11, Alpha=1, Translation="0,0,0", Rotation="0,0,0", Scale="0.6,1,1") {
    scaleMode=BMM_Additive
    scaleSpace=BCS_BoneSpace
    alphaBoolBlend {blendOption=Linear}
    boneToModify {boneName=dyn_backHairB_02}
    scale {x=0.6}
   }
   Save cached pose 'dynamicCache'(Pose=Blend Poses by bool(
     BlendPose_0=Component To Local(ComponentPose=@n12)
     BlendPose_1=Component To Local(ComponentPose=@n6)
     bActiveValue=Get Is Wearing Backbling()
    ) {blendType=Linear}) {cacheName=dynamicCache}
   Use cached pose 'testCache'()
  ]}
 EventGraph {nodes [
   n1 = Event BlueprintUpdateAnimation()
   > n2 = Sequence!(execute) -> (then_0, then_1, then_2:Exec, then_3:Exec, then_4:Exec, then_5:Exec, then_6:Exec, then_7:Exec)
   n3 = Gravity Override State Function DEPRECATED(
    execute=@n2.then_1
    CharacterPartAnimInstance=Self-Reference()
    GameTimeInSeconds=@n1.DeltaTimeX
    GravityOverrideParams=Get Gravity Override Params()
   )
   > Set Gravity Override(Gravity Override=@n3)
   Set head rot back(
    execute=@n2.then_3
    head rot back:float=GetAlphaFromReferenceTransformDelta(SkeletalMeshComponent=Get Body Mesh(), BoneName=head, BaseName=spine_05, InputRange=(X=-30,Y=0), OutputRange=(X=1,Y=0), TwistAxis="1,0,0")
   ) -> (Output_Get:float)
   n4 = Branch(execute=@n2.then_2, Condition=Get bIsPlayingEmote())
   > Set headRotScale(headRotScale:Rotator="1,1,1") -> (Output_Get:Rotator)
   Set headRotScale(execute=@n4.else, headRotScale="0.1,0.3,0.1")
   n5 = CheckForBackbling(self=Self-Reference())
   n6 = Branch(execute=@n2.then_5, Condition=OR Boolean!(A=@n5, B=@n5.bIsBackblingACape, C:bool=@n5.bIsBackblingLong) -> (ReturnValue))
   > Set Is Wearing Backbling(Is Wearing Backbling=true)
   Set Is Wearing Backbling(execute=@n6.else, Is Wearing Backbling=false)
   n7 = Scale Joint Rotation Function!(
    execute=@n2.then_6
    Scale:Rotator=Get headRotScale() -> (headRotScale:Rotator)
    TargetBone:Name=head
    ReferenceBone:Name=neck_01
    SkelMeshComponent:Skinned Mesh Component Object Reference=GetPartSkeletalMeshComponent()
   ) -> (then, FinalRotation:Rotator, OriginalRotation:Rotator)
   > Set head rot(head rot:Rotator=@n7) -> (Output_Get:Rotator)
   n8 = Get bIsPlayingEmote()
   n9 = Branch(execute=@n2.then_7, Condition=AND Boolean(A=NOT Boolean(A=@n8), B=Get bIsInFrontEnd()))
   > Set FrontEndConstraintToggle(FrontEndConstraintToggle:bool=true) -> (Output_Get:bool)
   Set FrontEndConstraintToggle(execute=@n9.else, FrontEndConstraintToggle=false)
   Event OnCharacterCustomizationLoaded()
   > n10 = Cast To PlayerPawn_Athena(Object=TryGetPawnOwner())
   > Set Body Mesh(Body Mesh:Skeletal Mesh Component Object Reference=GetSkeletalMeshForPartType(self=@n10)) -> (Output_Get:Skeletal Mesh Component Object Reference)
   OR Boolean!(A=false, B=Get isFrontEndIdle() -> (isFrontEndIdle:bool)) -> (ReturnValue)
  ]}
]
variables [
 Body Mesh {type=SkeletalMeshComponent}
 FrontEndConstraintToggle {type=boolean}
 Gravity Override {type=Vector}
 Gravity Override Params {type=GravityOverrideParamsStruct}
 head {type=Rotator}
 head rot {type=Rotator}
 head rot back {type=number}
 head Rot Right {type=number}
 headRotPercentageX {type=number}
 headRotPercentageY {type=number}
 headRotPercentageZ {type=number}
 headRotScale {type=Rotator}
 Is Wearing Backbling {type=boolean}
 isFrontEndIdle {type=boolean}
 isShotgun {type=boolean}
]
